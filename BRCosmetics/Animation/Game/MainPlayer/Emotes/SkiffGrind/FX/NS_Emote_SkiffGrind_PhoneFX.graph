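# Niagara system for the phone effects of the SkiffGrind emote: drifting little green hearts, pixel bursts and a bubble pop.
# LittleHearts spawns at a user-driven rate and records particle deaths in a system position array.
# System-level ReadDeath and PixelBurst's per-frame spawn turn those deaths into small sprite bursts, without Niagara events.
# BubblePop fires a timed burst at the gadget mesh's Bubble Socket; the flash and screen-light emitters have no renderers here.
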NiagaraSystem /BRCosmetics/Animation/Game/MainPlayer/Emotes/SkiffGrind/FX/NS_Emote_SkiffGrind_PhoneFX
dependencies {
 usedDataInterfaces [
  &NiagaraDataInterfaceCurve
  &NiagaraDataInterfaceVector2DCurve
  &NiagaraDataInterfaceColorCurve
  &NiagaraDataInterfaceSkeletalMesh
  &NiagaraDataInterfaceArrayPosition
 ]
 usedDynamicInputs [
  &Multiply_Int
  &MakeIntFromFloat
  &Multiply_LinearColorByFloat
 ]
 usedModules [
  &self:SystemSpawnScript.SystemScriptSource.SystemScriptGraph.NiagaraNodeAssignment_0.SetVariables_#1
  &SystemState
  &self:SystemSpawnScript.SystemScriptSource.SystemScriptGraph.NiagaraNodeAssignment_1.SetVariables_#2
  &self:ReadDeath
  &self:SystemSpawnScript.SystemScriptSource.SystemScriptGraph.NiagaraNodeAssignment_2.SetVariables_#3
  &SampleSkeletalMesh
 ]
}
emitters {
 BubblePop {
  emitterUpdateScript [
   &EmitterState()
   &SpawnBurst_Instantaneous(Spawn Count=$System.Flash Star Spawn, Spawn Time=5.62, Use Spawn Probability=false, UseLoopCountLimit=false)
  ]
  particleSpawnScript [
   &InitializeParticle(
    Color=$System.Bubble Pop Color TOD
    Lifetime=0.12
    Sprite Size Mode=::NewEnumerator3
    Uniform Sprite Size=35
    UsePositionOffset=false
    Write Material Random=true
   )
   &CameraOffset(Camera Offset Amount=2)
   &SkeletalMeshLocation(
    MaskSpawnGroup=false
    Mesh Sampling Type=::NewEnumerator5
    Sampled Position Offset=<-6 -4 0>
    Skeletal Mesh=DataInterface() {propertyValues {
      DefaultMesh=null
      ExcludeBoneName=root
      OptionalComponentClass=null
      PreviewMesh=/BRCosmetics/Animation/Props/Gadgets_SkiffGrind/Meshes/SkiffGrind_Gadget
      SkinningMode=SkinOnTheFly
      SoftSourceActor=null
      SourceMode=Default
      UvSetIndex=0
      WholeMeshLOD=-1
      bExcludeBone=false
      bReadDeformedGeometry=true
      bRequireCurrentFrameData=true
      ComponentTags []
      FilteredBones []
      FilteredSockets [Bubble Socket]
      MeshUserParameter {parameter {name=None, typeDefHandle {registeredTypeIndex=1}, varData []}}
      SamplingRegions []
     }}
    Socket Sampling Mode=::NewEnumerator5
   )
  ]
  particleUpdateScript [
   &ParticleState()
   &ScaleSpriteSize(Uniform Curve Sprite Scale=DataInterface() {propertyValues {Curve {keys [{arriveTangent=0.7338, interpMode=RCIM_Cubic, leaveTangent=0.7338, tangentMode=RCTM_User, time=0, value=0.8}, <at 1 1.2 RCIM_Cubic>]}}})
   &ScaleColor(
    Scale Alpha=&FloatFromCurve(FloatCurve=DataInterface() {propertyValues {Curve {keys [<at 0 1 RCIM_Cubic>, {arriveTangent=-2.4022, interpMode=RCIM_Cubic, leaveTangent=-2.4022, tangentMode=RCTM_User, time=1, value=0}]}}})
    ScaleA=true
    ScaleRGB=true
   )
  ]
  rendererClasses [&NiagaraSpriteRendererProperties]
  renderers [{rendererClass=&NiagaraSpriteRendererProperties}]
 }
 FlashCone {rendererClasses [], renderers []}
 FlashStar {rendererClasses [], renderers []}
 FlashStar_Light {rendererClasses [], renderers []}
 LittleHearts {
  emitterUpdateScript [
   &EmitterState()
   &SpawnRate(Spawn Probability=0.5, SpawnGroup=1, SpawnRate=$User.Idle_Loop_Additive, Use Spawn Probability=true)
  ]
  particleSpawnScript [
   &InitializeParticle(
    Color=&RandomRangeLinearColor(Minimum=<rgba 0 1 0.4148 1>, Maximum=<rgba 0.3177 1 0.6007 1>)
    Lifetime=&RandomRangeFloat(Minimum=0.35, Maximum=0.6)
    Mesh Scale=&VectorFromFloat(Value=&RandomRangeFloat(Minimum=0.05, Maximum=0.18))
    Mesh Scale Mode=::NewEnumerator4
    Sprite Size Mode=::NewEnumerator3
    Uniform Sprite Size=&RandomRangeFloat(Minimum=2, Maximum=4)
    UsePositionOffset=false
    Write Material Random=true
   )
   &AddVelocity(
    Rotation Coordinate Space=::Simulation
    Velocity=&RandomRangeVector(Minimum=<-55 -55 95>, Maximum=<55 55 115>)
   )
  ]
  particleUpdateScript [
   &ParticleState()
   &ScaleSpriteSize(
    Non-Uniform Curve Sprite Scale=DataInterface() {propertyValues {
      XCurve {keys [{tangentMode=RCTM_User, time=0, value=0}, {tangentMode=RCTM_User, time=0.2, value=1}, <at 0.9 1>, <at 1 2>]}
      YCurve {keys [
        <at 0 0 RCIM_Cubic>
        {arriveTangent=5.5116, interpMode=RCIM_Cubic, leaveTangent=5.5116, time=0.1, value=2}
        {arriveTangent=-2.2266, interpMode=RCIM_Cubic, leaveTangent=-2.2266, time=0.2041, value=1.125}
        {arriveTangent=-0.5135, interpMode=RCIM_Cubic, leaveTangent=-0.5135, time=0.3246, value=1.5}
        {arriveTangent=-0.9123, interpMode=RCIM_Cubic, leaveTangent=-0.9123, time=0.4476, value=1}
        {arriveTangentWeight=0.3037, interpMode=RCIM_Cubic, leaveTangentWeight=0.0772, tangentMode=RCTM_User, time=0.9, value=0.975}
        <at 1 2 RCIM_Cubic>
       ]}
     }}
    Scale Sprite Size Mode=::NewEnumerator3
   )
   &Drag(Drag=&RandomRangeFloat(Minimum=4, Maximum=8), Use Linear Drag=true, Use Rotational Drag=true)
   &AccelerationForce(Acceleration=&RandomRangeVector(Minimum=<0 0 30>, Maximum=<0 0 60>))
   &SolveForcesAndVelocity(Clamp Velocity=false, Limit Acceleration=false)
   &ScaleColor(
    Scale Alpha=&FloatFromCurve(
     FloatCurve=DataInterface() {propertyValues {Curve {keys [<at 0 1>,<at 0.7974 1>,<at 1 0>]}}}
     Scale Curve=&FNTimeOfDayScalar(Night Scale=5, Morning Scale=1, Day Scale=5, Evening Scale=1)
    )
    ScaleA=true
    ScaleRGB=true
   )
   &DynamicMaterialParameters(
    Index 0 Param 2=1
    Module.Index0Float0Write=true
    Module.Index0Float1Write=true
    Module.Index0Float2Write=true
    Module.Index0Float3Write=true
   )
   &self:WriteDeath(Position Array:NiagaraDataInterfaceArrayPosition=$System.DeathArray) {module=WriteDeath}
  ]
  rendererClasses [&NiagaraSpriteRendererProperties]
  renderers [{rendererClass=&NiagaraSpriteRendererProperties}]
 }
 PixelBurst {
  emitterUpdateScript [
   &EmitterState()
   &SpawnPerFrame(
    Spawn:NiagaraBool=$System.DeathSpawn
    Spawn Count:NiagaraInt32=&RandomRangeInt(Minimum=4, Maximum=5)
    Use Spawn Probability=false
   )
  ]
  particleSpawnScript [
   &InitializeParticle(
    Color=<rgba 0.4792 1 0.7528 1>
    Lifetime Max=0.4
    Lifetime Min=0.25
    Lifetime Mode=::NewEnumerator1
    Position=$System.DeathPosition
    Position Mode=::NewEnumerator1
    Sprite Size Mode=::NewEnumerator1
    Uniform Sprite Size Max=1.6
    Uniform Sprite Size Min=0.8
    UsePositionOffset=false
    Write Material Random=true
   )
   &AddVelocity(
    Cone Angle=360
    Distribution Along Cone Axis=0.2
    Module.UseConeFalloff=false
    Module.UseVelDistribution=true
    Override Seed=false
    Velocity Mode=::NewEnumerator2
    Velocity Speed=&RandomRangeFloat(Minimum=25, Maximum=50)
   )
  ]
  particleUpdateScript [
   &ParticleState()
   &ScaleSpriteSize(
    Uniform Curve Scale=&Waveform("Waveform [1]"=::NewEnumerator0, "[1] Frequency"=1, "[1] Amplitude Min/Max"=<0.7 1>)
    Uniform Curve Sprite Scale=DataInterface() {propertyValues {Curve {keys [
        {arriveTangent=3.1789, interpMode=RCIM_Cubic, leaveTangent=3.1789, tangentMode=RCTM_User, time=0, value=0.7}
        {
         arriveTangent=-0.6941
         arriveTangentWeight=0.2414
         interpMode=RCIM_Cubic
         leaveTangent=-0.6941
         leaveTangentWeight=0.1655
         tangentMode=RCTM_User
         time=0.5932
         value=1
        }
        <at 1 0 RCIM_Cubic>
       ]}}}
   )
   &Drag(
    Drag=&FloatFromCurve(
     FloatCurve=DataInterface() {propertyValues {Curve {keys [{arriveTangent=3.209, interpMode=RCIM_Cubic, leaveTangent=3.209, tangentMode=RCTM_User, time=0, value=0}, <at 1 1 RCIM_Cubic>]}}}
     Scale Curve=10
    )
    Use Linear Drag=true
    Use Rotational Drag=true
   )
   &SolveForcesAndVelocity(Clamp Velocity=false, Limit Acceleration=false)
   &ScaleColor(
    Linear Color Curve=DataInterface() {propertyValues {AlphaCurve {keys [<at 0 1>,<at 0.5806 1>,<at 1 0>]}, BlueCurve {keys [<at 0 5>]}, GreenCurve {keys [<at 0 5>]}, RedCurve {keys [<at 0 5>]}}}
    Scale Mode=2
   )
  ]
  rendererClasses [&NiagaraSpriteRendererProperties]
  renderers [{rendererClass=&NiagaraSpriteRendererProperties}]
 }
 Screen_Light {rendererClasses [], renderers []}
}
summary {
 systemName=NS_Emote_SkiffGrind_PhoneFX
 emitters [
  {emitterName=LittleHearts, rendererClasses [&NiagaraSpriteRendererProperties]}
  {emitterName=FlashStar, rendererClasses []}
  {emitterName=FlashStar_Light, rendererClasses []}
  {emitterName=Screen_Light, rendererClasses []}
  {emitterName=PixelBurst, rendererClasses [&NiagaraSpriteRendererProperties]}
  {emitterName=FlashCone, rendererClasses []}
  {emitterName=BubblePop, rendererClasses [&NiagaraSpriteRendererProperties]}
 ]
 userVariables [
  User.Balloon Color {type=LinearColor, description="", defaultValue {struct=LinearColor, value=<rgba 0.2717 0.0264 0.7 1>}}
  User.bIsFrontEnd {type=NiagaraBool}
  User.bIsFrontEndPreview {type=NiagaraBool}
  User.BurstIndex {type=NiagaraFloat, defaultValue=1, description=""}
  User.Idle_Loop_Additive {type=NiagaraFloat, defaultValue=1, description=""}
  User.NormalizedNotifyProgress {type=NiagaraFloat, defaultValue=0, description=""}
 ]
}
