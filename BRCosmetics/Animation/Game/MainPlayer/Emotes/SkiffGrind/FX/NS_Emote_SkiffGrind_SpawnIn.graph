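# Niagara system for the SkiffGrind emote's spawn-in: a brief pink-tinted descending streak and ground impact effect.
# AirLines drops streaks and glows, while ImpactFlare and Shockwave follow slightly delayed with flares, light, ripple, decal and time-of-day-tinted dust meshes.
# Decal adds a flickering, fading ground mark, and Smoke bursts smoke, streaks and motes.
# Spawn groups set per-particle behavior and renderer visibility tags; BlendSpace selects world/local blending from owner speed.
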NiagaraSystem /BRCosmetics/Animation/Game/MainPlayer/Emotes/SkiffGrind/FX/NS_Emote_SkiffGrind_SpawnIn
dependencies {
 usedDataInterfaces [
  &NiagaraDataInterfaceCurve
  &NiagaraDataInterfaceVectorCurve
  &NiagaraDataInterfaceColorCurve
  &NiagaraDataInterfaceVector2DCurve
  &NiagaraDataInterfaceArrayColor
 ]
 usedDynamicInputs []
 usedModules [&SystemState, &BlendSpaceHelper]
}
emitters {
 AirLines {
  emitterUpdateScript [
   &EmitterState(Life Cycle Mode=::NewEnumerator1, Loop Behavior=::NewEnumerator1, Loop Duration=0.35, UseLoopDelay=false)
   &SpawnBurst_Instantaneous(Spawn Group=3, Use Spawn Probability=false, UseLoopCountLimit=false) {module=SpawnBurst_Instantaneous002}
   &SpawnBurst_Instantaneous(Spawn Group=1, Use Spawn Probability=false, UseLoopCountLimit=false) {module=SpawnBurst_Instantaneous001}
   &SpawnBurst_Instantaneous(Spawn Count=5, Spawn Time=0.05, Use Spawn Probability=false, UseLoopCountLimit=false)
   &SpawnRate(
    SpawnRate=&FloatFromCurve(
     FloatCurve=DataInterface() {propertyValues {Curve {keys [<at 0 1 RCIM_Cubic>, {arriveTangent=-2.82, interpMode=RCIM_Cubic, leaveTangent=-2.82, tangentMode=RCTM_User, time=0.6, value=0}]}}}
     CurveIndex=$Emitter.NormalizedLoopAge
     Scale Curve=30
    )
    Use Spawn Probability=false
   )
   &SpawnRate(
    SpawnGroup=2
    SpawnRate=&FloatFromCurve(
     FloatCurve=DataInterface() {propertyValues {Curve {keys [<at 0 1 RCIM_Cubic>,<at 0.75 0 RCIM_Cubic>]}}}
     CurveIndex=$Emitter.NormalizedLoopAge
     Scale Curve=30
    )
    Use Spawn Probability=false
   ) {module=SpawnRate001}
  ]
  particleSpawnScript [
   &InitializeParticle(
    Color=&RandomRangeLinearColor(Minimum=<rgba 0.4635 0.1146 0.2134 1>, Maximum=<rgba 1 0.4427 0.683 1>)
    Lifetime=&MergeHelper_FloatSwitch(
     0=&RandomRangeFloat(Minimum=0.1, Maximum=0.45)
     1=0.18
     2=&RandomRangeFloat(Minimum=0.16, Maximum=0.26)
     3=0.5
     Index=&SpawnGroup()
     Num=::NewEnumerator2
    )
    Mesh Scale=<1 1 1.6>
    Mesh Scale Mode=::NewEnumerator4
    Position Offset=&MergeHelper_VectorSwitch(0=<0 0 150>, 1=<0 0 350>, 2=<0 0 280>, 3=<0 0 -62.9355>, Index=&SpawnGroup(), Num=::NewEnumerator2)
    Sprite Rotation Angle=180
    Sprite Rotation Mode=::NewEnumerator0
    Sprite Size Max=<7 400>
    Sprite Size Min=<1 300>
    Sprite Size Mode=::NewEnumerator2
    UsePositionOffset=true
    Write Material Random=true
   )
   &AddVelocity(Velocity=&MergeHelper_VectorSwitch(
     0=<0 0 0>
     1=<0 0 -2000>
     2=&RandomRangeVector(Minimum=<0 0 -1000>, Maximum=<1 1 -1500>)
     3=?
     Index=&SpawnGroup()
     Num=::NewEnumerator1
    ))
   &ShapeLocation(
    Cylinder Height=50
    Cylinder Radius=35
    MaskSpawnGroup=true
    Offset=<0 0 20>
    Offset Mode=::NewEnumerator0
    Override Seed=false
    Shape Primitive=::NewEnumerator1
    Surface Only=false
    UseOwnerScale=true
   )
   &ShapeLocation(
    Cylinder Height=50
    Cylinder Radius=30
    MaskSpawnGroup=true
    Offset=<0 0 20>
    Offset Mode=::NewEnumerator0
    Override Seed=false
    Shape Primitive=::NewEnumerator1
    Spawn Group Mask=2
    Surface Only=false
    UseOwnerScale=true
   ) {module=ShapeLocation001}
   /SetVariables(Particles.VisibilityTag=&SpawnGroup()) {isSetParameters=true}
   /SetVariables(Particles.Glow Color=<rgba 1 0.4948 0.7095 1>, Particles.Glow Sprite Size:Vector2f=<100 300>) {isSetParameters=true}
   /SetVariables(
    Particles.Mote Color:LinearColor=&RandomRangeLinearColor(Minimum=<rgba 1 0.6333 0.7914 1>, Maximum=<rgba 1 0.2471 0.4603 1>)
    Particles.Mote Sprite Size:Vector2f=&RandomRangeVector2D(Minimum=<8 150>, Maximum=<15 200>)
   ) {isSetParameters=true}
   /SetVariables(Particles.Line Glow Color:LinearColor=<rgba 1 0.4427 0.683 0.5>) {isSetParameters=true}
  ]
  particleUpdateScript [
   &ParticleState()
   &ScaleSpriteSize(
    Non-Uniform Curve Sprite Scale=DataInterface() {propertyValues {
      XCurve {keys [{tangentMode=RCTM_User, time=0, value=0.8}, {tangentMode=RCTM_User, time=1, value=1}]}
      YCurve {keys [{tangentMode=RCTM_User, time=0, value=1}, {tangentMode=RCTM_User, time=1, value=1}]}
     }}
    Scale Sprite Size Mode=::NewEnumerator3
   )
   &BlendSpace("Position Blend (World <-> Local)":NiagaraFloat=&MakeCustomFloatFromBool(
     Bool=&SetBoolByFloatComparison(A=&VectorLength(Vector=$Engine.Owner.Velocity), B=400)
     True Float=1
     False Float=0
     Translation Contribution:NiagaraFloat=1
     Rotation Contribution:NiagaraFloat=&MakeCustomFloatFromBool(
      Bool=&SetBoolByFloatComparison(A=&VectorLength(Vector=$Engine.Owner.Velocity), B=800)
      True Float=1
      False Float=0
     )
    )) {module=BlendSpace002}
   &SolveForcesAndVelocity(Clamp Velocity=false, Limit Acceleration=false)
   &ScaleColor(
    Linear Color Curve=DataInterface() {propertyValues {
      AlphaCurve {keys [<at 0 0>,<at 0.1977 1>,<at 0.4082 1>,<at 1 0>]}
      BlueCurve {keys [<at 0 5>,<at 0.3246 3>,<at 1 1>]}
      GreenCurve {keys [<at 0 5>,<at 0.3246 3>,<at 1 1>]}
      RedCurve {keys [<at 0 5>,<at 0.3246 3>,<at 1 1>]}
     }}
    Scale Mode=::NewEnumerator2
   )
   /SetVariables(
    Particles.Glow Color=&MultiplyLinearColor(
     Linear Color=$Particles.Initial.Glow Color
     Scale Factor=&MakeVector4FromVectorAndFloat(
      "Vector (RGB)"=<1 1 1>
      "Float (Alpha)"=&FloatFromCurve(FloatCurve=DataInterface() {propertyValues {Curve {keys [<at 0.2994 1 RCIM_Cubic>,<at 1 0.3 RCIM_Cubic>]}}})
     )
    )
    Particles.Glow Sprite Size:Vector2f=&Multiply_Vector2D(
     A=$Particles.Initial.Glow Sprite Size
     B=&Vector2DFromCurve(Vector2Curve=DataInterface() {propertyValues {XCurve {keys [<at 0 1>,<at 1 1>]}, YCurve {keys [<at 0 1>,<at 1 1.5>]}}}, ScaleCurve=<1 1>)
    )
   ) {isSetParameters=true}
   /SetVariables(
    Particles.Mote Color:LinearColor=&MultiplyLinearColor(
     Linear Color=$Particles.Initial.Mote Color
     Scale Factor=&MakeVector4FromVectorAndFloat(
      "Vector (RGB)"=<1 1 1>
      "Float (Alpha)"=&FloatFromCurve(FloatCurve=DataInterface() {propertyValues {Curve {keys [<at 0.2994 1 RCIM_Cubic>, {arriveTangent=-1.7829, interpMode=RCIM_Cubic, leaveTangent=-1.7829, tangentMode=RCTM_User, time=1, value=0.3}]}}})
     )
    )
    Particles.Mote Sprite Size:Vector2f=&Multiply_Vector2D(
     A=$Particles.Initial.Mote Sprite Size
     B=&Vector2DFromCurve(
      Vector2Curve=DataInterface() {propertyValues {XCurve {keys [<at 0 1>,<at 1 1>]}, YCurve {keys [{arriveTangent=0.5, interpMode=RCIM_Cubic, time=0, value=1}, <at 1 0.5 RCIM_Cubic>]}}}
      ScaleCurve=<1 2>
     )
    )
   ) {isSetParameters=true}
   /SetVariables(Particles.Line Glow Color:LinearColor=&MultiplyLinearColor(
     Linear Color=$Particles.Initial.Line Glow Color
     Scale Factor=&Vector4FromCurve(Vector4Curve=DataInterface() {propertyValues {
        AlphaCurve {keys [<at 0.0498 0>,<at 0.22 1>,<at 1 0>]}
        BlueCurve {keys [<at 0 1>,<at 1 1>]}
        GreenCurve {keys [<at 0 1>,<at 1 1>]}
        RedCurve {keys [<at 0 1>,<at 1 1>]}
       }})
    )) {isSetParameters=true}
  ]
  rendererClasses [&NiagaraSpriteRendererProperties, &NiagaraMeshRendererProperties]
  renderers [
   {rendererClass=&NiagaraSpriteRendererProperties}
   {rendererClass=&NiagaraSpriteRendererProperties, rendererIndex=1}
   {rendererClass=&NiagaraSpriteRendererProperties, rendererIndex=2}
   {rendererClass=&NiagaraMeshRendererProperties, rendererIndex=3}
  ]
 }
 Decal {
  emitterUpdateScript [
   &EmitterState()
   &SpawnBurst_Instantaneous(Use Spawn Probability=false, UseLoopCountLimit=false)
  ]
  particleSpawnScript [
   &InitializeParticle(
    Color=<rgba 2 0.4583 1.1135 1>
    Lifetime=1.5
    Mesh Scale Mode=::NewEnumerator4
    Sprite Size Mode=::NewEnumerator3
    UsePositionOffset=false
    Write Material Random=true
   )
   &InitialMeshOrientation(
    Mesh Orientation Mode=::NewEnumerator2
    Rotation=&RandomRangeVector(Minimum=<0 0 0>, Maximum=<0 0 1>)
    Use Rotation Vector=false
   )
  ]
  particleUpdateScript [
   &ParticleState()
   &BlendSpace("Position Blend (World <-> Local)":NiagaraFloat=&MakeCustomFloatFromBool(
     Bool=&SetBoolByFloatComparison(A=&VectorLength(Vector=$Engine.Owner.Velocity), B=400)
     True Float=1
     False Float=0
     Translation Contribution:NiagaraFloat=1
     Rotation Contribution:NiagaraFloat=&MakeCustomFloatFromBool(
      Bool=&SetBoolByFloatComparison(A=&VectorLength(Vector=$Engine.Owner.Velocity), B=800)
      True Float=1
      False Float=0
     )
    )) {module=BlendSpace002}
   /Niagara/Modules/Decal/Decal_Attributes(
    Color=1
    Decal Alignment:Vector3f=<1 0 0>
    Decal Facing Mode=1
    Decal Rotation Mode=2
    Decal Size:Vector3f=<50 200 200>
    Fade=2
    Orientation=2
    Position=1
    Scale Alpha:NiagaraFloat=&FloatFromCurve(
     FloatCurve=DataInterface() {propertyValues {Curve {keys [<at 0 1>,<at 0.0526 0.2>,<at 0.117 0.85>,<at 0.1832 0.25>,<at 0.2457 0.65>,<at 1 0>]}}}
     Scale Curve=0.5
    )
    ScaleA=true
    ScaleRGB=true
    Set Alignment=true
    Size=2
    UsePositionOffset=false
    Visible=2
   )
   &DynamicMaterialParameters(Write Parameter Index 0=::NewEnumerator0)
  ]
  rendererClasses [&NiagaraDecalRendererProperties]
  renderers [{rendererClass=&NiagaraDecalRendererProperties, rendererIndex=0}]
 }
 ImpactFlare {
  emitterUpdateScript [
   &EmitterState(Life Cycle Mode=::NewEnumerator1, Loop Behavior=::NewEnumerator1, Loop Delay=0.1, Loop Duration=2, UseLoopDelay=true)
   &SpawnBurst_Instantaneous(Spawn Group=3, Use Spawn Probability=false, UseLoopCountLimit=false) {module=SpawnBurst_Instantaneous002}
   &SpawnBurst_Instantaneous(Use Spawn Probability=false, UseLoopCountLimit=false) {module=SpawnBurst_Instantaneous003}
   &SpawnBurst_Instantaneous(Spawn Group=1, Use Spawn Probability=false, UseLoopCountLimit=false) {module=SpawnBurst_Instantaneous001}
   &SpawnBurst_Instantaneous(Spawn Group=2, Use Spawn Probability=false, UseLoopCountLimit=false)
  ]
  particleSpawnScript [
   /SetVariables(Particles.VisibilityTag=&SpawnGroup(Max Spawn Group=5)) {isSetParameters=true}
   &InitializeParticle(
    Color=&MakeCustomLinearColorFromBool(
     Bool=&SetBoolByIntComparison(A=&SpawnGroup(), B=1, Comparison Type=::NewEnumerator7)
     True Color=&SelectLinearColorFromArray(
      Array Sampling Mode=::NewEnumerator1
      Direct Array Index=&SpawnGroup()
      Color Selection Array=DataInterface() {propertyValues {ColorData [<rgba 1 1 1 1>, <rgba 1 1 1 0.5>, <rgba 1 1 1 1>, <rgba 100 61.7074 35.1155 1>]}}
      Random Seed=?
     )
     False Color=<rgba 1 1 1 1>
    )
    Lifetime=&MergeHelper_FloatSwitch(0=0.5, 1=0.4, 2=0.25, 3=1.7, Index=&SpawnGroup(), Num=::NewEnumerator2)
    Mesh Scale Mode=::NewEnumerator3
    Mesh Uniform Scale=&MaskFloatBySpawnGroup(Masked Float=1, Passthrough Float=1, Spawn Group Mask=2)
    Position Offset=<40 -20 150>
    Sprite Rotation Angle=&MaskFloatBySpawnGroup(Masked Float=0, Passthrough Float=&RandomRangeFloat(Minimum=0, Maximum=1))
    Sprite Rotation Mode=::NewEnumerator1
    Sprite Size=&Vector2DFromFloat(Value=250)
    Sprite Size Mode=::NewEnumerator4
    UsePositionOffset=false
    Write Material Random=true
   )
   &CameraOffset(Camera Offset Amount=&MergeHelper_FloatSwitch(0=50, 1=20, 2=0, 3=?, Index=&SpawnGroup(), Num=::NewEnumerator1))
   /SetVariables(
    Particles.Orange Glow Scale:Vector3f=&VectorFromFloat(Value=4)
    Particles.Orange Glow Scale 2:Vector3f=&VectorFromFloat(Value=2)
   ) {isSetParameters=true}
   /SetVariables(Particles.Impact Color:LinearColor=<rgba 0.1406 0.1406 0.1406 1>) {isSetParameters=true}
   /SetVariables(
    Particles.Ripple Color:LinearColor=<rgba 1 0.1458 0.3068 1>
    Particles.Ripple Sprite Size:Vector2f=&Vector2DFromFloat(Value=300)
   ) {isSetParameters=true}
  ]
  particleUpdateScript [
   &ParticleState()
   &ScaleSpriteSize(
    Scale Factor=&MaskVector2dBySpawnGroup(
     Masked Vector2D=&Vector2DFromCurve(
      Vector2Curve=DataInterface() {propertyValues {
        XCurve {keys [<at 0 1 RCIM_Cubic>, {arriveTangent=-3.1442, interpMode=RCIM_Cubic, leaveTangent=-3.1442, tangentMode=RCTM_User, time=1, value=0}]}
        YCurve {keys [
          {arriveTangent=4.8145, interpMode=RCIM_Cubic, time=0, value=1}
          {arriveTangent=-3.0695, interpMode=RCIM_Cubic, leaveTangent=-3.0695, tangentMode=RCTM_User, time=1, value=0}
         ]}
       }}
      ScaleCurve=<1 1>
     )
     Passthrough Vector2D=&Vector2DFromFloat(Value=&MaskFloatBySpawnGroup(
       Masked Float=&FloatFromCurve(FloatCurve=DataInterface() {propertyValues {Curve {keys [
            {arriveTangent=3.8383, interpMode=RCIM_Cubic, leaveTangent=3.8383, tangentMode=RCTM_User, time=0, value=0}
            {arriveTangentWeight=0.1274, interpMode=RCIM_Cubic, leaveTangentWeight=0.7778, tangentMode=RCTM_User, time=0.5, value=1}
           ]}}})
       Passthrough Float=&FloatFromCurve(FloatCurve=DataInterface() {propertyValues {Curve {keys [
            {arriveTangent=3.3451, interpMode=RCIM_Cubic, leaveTangent=3.3451, tangentMode=RCTM_User, time=0, value=0.5}
            {arriveTangentWeight=0.1274, interpMode=RCIM_Cubic, leaveTangentWeight=0.7778, tangentMode=RCTM_User, time=0.5, value=1}
           ]}}})
       Spawn Group Mask=1
      ))
     Spawn Group Mask=0
    )
    Scale Sprite Size Mode=::NewEnumerator2
   )
   &Color(Color=&MakeCustomLinearColorFromBool(
     Bool=&SetBoolByIntComparison(A=&SpawnGroup(), B=1, Comparison Type=::NewEnumerator7)
     True Color=&SelectLinearColorFromArray(
      Array Sampling Mode=::NewEnumerator1
      Direct Array Index=&SpawnGroup()
      Color Selection Array=DataInterface() {propertyValues {ColorData [<rgba 1 1 1 1>, <rgba 1 1 1 0.5>, <rgba 1 1 1 1>, <rgba 100 61.7074 35.1155 1>]}}
      Random Seed=?
     )
     False Color=<rgba 1 1 1 1>
    ))
   &Decal_Attributes(
    Color=1
    Color Value To Scale:LinearColor=$Particles.Initial.Impact Color
    Decal Alignment:Vector3f=<1 0 0>
    Decal Facing Mode=1
    Decal Rotation Mode=2
    Decal Size:Vector3f=<50 250 250>
    Decal Sort Order:NiagaraInt32=&SpawnGroup()
    Fade=2
    Orientation=2
    Position=1
    Scale Alpha:NiagaraFloat=&FloatFromCurve(FloatCurve=DataInterface() {propertyValues {Curve {keys [<at 0.5002 1>,<at 0.7459 0.65>,<at 1 0>]}}})
    Scale RGB:Vector3f=&MakeVectorFromLinearColorRGB(Color=&ColorFromCurve(ColorCurve=DataInterface() {propertyValues {
        AlphaCurve {keys []}
        BlueCurve {keys [<at 0 0 RCIM_Cubic>,<at 0.5112 1 RCIM_Cubic>]}
        GreenCurve {keys [<at 0 0.8242 RCIM_Cubic>,<at 0.5112 1 RCIM_Cubic>]}
        RedCurve {keys [<at 0 5 RCIM_Cubic>,<at 0.5112 1 RCIM_Cubic>]}
       }}))
    ScaleA=true
    ScaleRGB=true
    Set Alignment=true
    Size=2
    Sort Order=1
    UsePositionOffset=false
    Visible=2
   )
   &BlendSpace("Position Blend (World <-> Local)":NiagaraFloat=&MakeCustomFloatFromBool(
     Bool=&SetBoolByFloatComparison(A=&VectorLength(Vector=$Engine.Owner.Velocity), B=400)
     True Float=1
     False Float=0
     Translation Contribution:NiagaraFloat=1
     Rotation Contribution:NiagaraFloat=&MakeCustomFloatFromBool(
      Bool=&SetBoolByFloatComparison(A=&VectorLength(Vector=$Engine.Owner.Velocity), B=800)
      True Float=1
      False Float=0
     )
    )) {module=BlendSpace002}
   &ScaleMeshSize(Scale Factor=&VectorFromFloat(Value=&MaskFloatBySpawnGroup(
      Masked Float=&FloatFromCurve(FloatCurve=DataInterface() {propertyValues {Curve {
          defaultValue=0
          keys [
           {arriveTangent=4.1785, interpMode=RCIM_Cubic, leaveTangent=4.1785, tangentMode=RCTM_User, time=0, value=0}
           {arriveTangent=1.529, interpMode=RCIM_Cubic, leaveTangent=1.529, tangentMode=RCTM_User, time=0.2042, value=0.6228}
           {arriveTangent=0.0224, interpMode=RCIM_Cubic, leaveTangent=0.0224, tangentMode=RCTM_User, time=1, value=1}
          ]
         }}})
      Passthrough Float=&FloatFromCurve(FloatCurve=DataInterface() {propertyValues {Curve {
          defaultValue=0
          keys [
           <at 0 0 RCIM_Cubic>
           {arriveTangent=1.6664, interpMode=RCIM_Cubic, leaveTangent=1.6664, tangentMode=RCTM_User, time=0.45, value=0.2}
           {arriveTangent=1.1179, interpMode=RCIM_Cubic, leaveTangent=1.1179, tangentMode=RCTM_User, time=0.55, value=0.75}
           {arriveTangent=0.0224, interpMode=RCIM_Cubic, leaveTangent=0.0224, tangentMode=RCTM_User, time=1, value=1}
          ]
         }}})
      Spawn Group Mask=2
     )))
   &DynamicMaterialParameters(
    Index 0 Param 1=&MaskFloatBySpawnGroup(
     Masked Float=&FloatFromCurve(FloatCurve=DataInterface() {propertyValues {Curve {keys [{arriveTangent=0.1648, interpMode=RCIM_Cubic, leaveTangent=0.1648, tangentMode=RCTM_User, time=0, value=0.9}, <at 0.5 1 RCIM_Cubic>]}}})
     Passthrough Float=&FloatFromCurve(FloatCurve=DataInterface() {propertyValues {Curve {keys [<at 0 0.9 RCIM_Cubic>,<at 1 1 RCIM_Cubic>]}}})
     Spawn Group Mask=2
    )
    Index 0 Param 3=5
    Index 0 Param 4=0.04
    Module.Index0Float0Write=true
    Module.Index0Float1Write=false
    Module.Index0Float2Write=true
    Module.Index0Float3Write=true
    Module.Index1Float0Write=true
    Module.Index1Float1Write=true
    Module.Index1Float2Write=true
    Module.Index1Float3Write=true
    Module.Index2Float0Write=true
    Module.Index2Float1Write=true
    Module.Index2Float2Write=true
    Module.Index2Float3Write=true
    Write Parameter Index 1=::NewEnumerator1
    Write Parameter Index 2=::NewEnumerator1
   )
   &ScaleColor(
    Scale Alpha=&MaskFloatBySpawnGroup(
     Masked Float=&FloatFromCurve(FloatCurve=DataInterface() {propertyValues {Curve {keys [
          {arriveTangent=-23.168, interpMode=RCIM_Cubic, leaveTangent=-23.168, tangentMode=RCTM_User, time=0, value=4}
          {arriveTangent=-4.5789, interpMode=RCIM_Cubic, leaveTangent=-4.5789, tangentMode=RCTM_User, time=0.2, value=1}
          <at 1 0 RCIM_Cubic>
         ]}}})
     Passthrough Float=&MaskFloatBySpawnGroup(
      Masked Float=&FloatFromCurve(
       FloatCurve=DataInterface() {propertyValues {Curve {
          defaultValue=0
          keys [
           {arriveTangent=9.0725, interpMode=RCIM_Cubic, leaveTangent=9.0725, tangentMode=RCTM_User, time=0, value=0}
           {arriveTangent=-2.5249, interpMode=RCIM_Cubic, leaveTangent=-2.5249, tangentMode=RCTM_User, time=0.1408, value=0.9847}
           {arriveTangent=-0.0405, interpMode=RCIM_Cubic, leaveTangent=-0.0405, tangentMode=RCTM_User, time=1, value=0}
          ]
         }}}
       Scale Curve=0.2
      )
      Passthrough Float=&MaskFloatBySpawnGroup(
       Masked Float=&FloatFromCurve(
        FloatCurve=DataInterface() {propertyValues {Curve {keys [<at 0 1 RCIM_Cubic>,<at 1 0 RCIM_Cubic>]}}}
        Scale Curve=&MaskFloatBySpawnGroup(Masked Float=0.2, Passthrough Float=1, Spawn Group Mask=2)
       )
       Passthrough Float=&FloatFromCurve(
        FloatCurve=DataInterface() {propertyValues {Curve {keys [<at 0 1 RCIM_Cubic>,<at 1 0 RCIM_Cubic>]}}}
        Scale Curve=&MaskFloatBySpawnGroup(Masked Float=0.2, Passthrough Float=1, Spawn Group Mask=2)
       )
       Spawn Group Mask=2
      )
      Spawn Group Mask=1
     )
    )
    Scale RGB=&MakeVectorFromLinearColorRGB(Color=&MakeCustomLinearColorFromBool(
      Bool=&SetBoolByIntComparison(A=&SpawnGroup(), B=1, Comparison Type=::NewEnumerator7)
      True Color=&ColorFromCurve(ColorCurve=DataInterface() {propertyValues {
         AlphaCurve {keys []}
         BlueCurve {keys [{arriveTangent=-1.5872, interpMode=RCIM_Cubic, leaveTangent=-1.5872, tangentMode=RCTM_User, time=0, value=3}, <at 1 0.5 RCIM_Cubic>]}
         GreenCurve {keys [{arriveTangent=-4.9882, interpMode=RCIM_Cubic, leaveTangent=-4.9882, tangentMode=RCTM_User, time=0, value=3.75}, <at 1 0.625 RCIM_Cubic>]}
         RedCurve {keys [{arriveTangent=-3.5785, interpMode=RCIM_Cubic, leaveTangent=-3.5785, tangentMode=RCTM_User, time=0, value=4}, <at 1 1 RCIM_Cubic>]}
        }})
      False Color=<rgba 1 1 1 1>
     ))
    ScaleA=true
    ScaleRGB=true
   )
   &Light_Attributes(
    Color=1
    Color Value To Scale:LinearColor=$Particles.Orange Glow Color
    Diffuse Scale=2
    Enabled=2
    Falloff Exponent=2
    Light Exponent=3
    Light Radius=300
    Light Specular Scale=0
    Position=1
    Radius=2
    Scale RGB:Vector3f=<5 5 5>
    ScaleA=true
    ScaleRGB=true
    Specular Scale=2
    UsePositionOffset=false
   )
   /SetVariables(
    Particles.Orange Glow Color:LinearColor=&ColorFromCurve(ColorCurve=DataInterface() {propertyValues {
       AlphaCurve {keys [<at 0 1>,<at 1 0>]}
       BlueCurve {keys [<at 0 0>,<at 0.313 0.7155>]}
       GreenCurve {keys [<at 0 1.4296>,<at 0.313 0.5052>]}
       RedCurve {keys [<at 0 10>,<at 0.313 1>]}
      }})
    Particles.Orange Glow Color 2:LinearColor=&ColorFromCurve(ColorCurve=DataInterface() {propertyValues {
       AlphaCurve {keys [<at 0 1>,<at 1 0>]}
       BlueCurve {keys [<at 0 0>,<at 0.313 2.1464>]}
       GreenCurve {keys [<at 0 1.4296>,<at 0.313 1.5156>]}
       RedCurve {keys [<at 0 10>,<at 0.313 3>]}
      }})
    Particles.Orange Glow Scale:Vector3f=&Multiply_Vector(
     A=$Particles.Initial.Orange Glow Scale
     B=&VectorFromFloat(Value=&FloatFromCurve(FloatCurve=DataInterface() {propertyValues {Curve {keys [<at 0 0.4>,<at 1 1.25>]}}}, Scale Curve=1.25))
    )
    Particles.Orange Glow Scale 2:Vector3f=&Multiply_Vector(
     A=$Particles.Initial.Orange Glow Scale 2
     B=&VectorFromFloat(Value=&FloatFromCurve(FloatCurve=DataInterface() {propertyValues {Curve {keys [<at 0 0.4>,<at 1 1.25>]}}}, Scale Curve=1.25))
    )
   ) {isSetParameters=true}
   /SetVariables(Particles.Shockwave Scale:Vector3f=&VectorFromFloat(Value=&FloatFromCurve(FloatCurve=DataInterface() {propertyValues {Curve {
         defaultValue=0
         keys [
          <at 0 0 RCIM_Cubic>
          {arriveTangent=1.6664, interpMode=RCIM_Cubic, leaveTangent=1.6664, tangentMode=RCTM_User, time=0.45, value=0.2}
          {arriveTangent=1.1179, interpMode=RCIM_Cubic, leaveTangent=1.1179, tangentMode=RCTM_User, time=0.55, value=0.75}
          {arriveTangent=0.0224, interpMode=RCIM_Cubic, leaveTangent=0.0224, tangentMode=RCTM_User, time=1, value=1}
         ]
        }}}))) {isSetParameters=true}
   /SetVariables(
    Particles.Ripple Color:LinearColor=&MultiplyLinearColor(
     Linear Color=$Particles.Initial.Ripple Color
     Scale Factor=&Vector4FromCurve(Vector4Curve=DataInterface() {propertyValues {
        AlphaCurve {keys [<at 0 1>,<at 1 0>]}
        BlueCurve {keys [<at 0 1>,<at 1 1>]}
        GreenCurve {keys [<at 0 1>,<at 1 1>]}
        RedCurve {keys [<at 0 1>,<at 1 1>]}
       }})
    )
    Particles.Ripple Sprite Size:Vector2f=&Multiply_Vector2D(
     A=$Particles.Initial.Ripple Sprite Size
     B=&Vector2DFromFloat(Value=&FloatFromCurve(
       FloatCurve=DataInterface() {propertyValues {Curve {keys [
           {arriveTangent=5.1291, interpMode=RCIM_Cubic, leaveTangent=5.1291, tangentMode=RCTM_User, time=0, value=0}
           {arriveTangentWeight=0.1274, interpMode=RCIM_Cubic, leaveTangentWeight=0.7778, tangentMode=RCTM_User, time=0.5, value=1}
          ]}}}
       Scale Curve=1.5
      ))
    )
   ) {isSetParameters=true}
  ]
  rendererClasses [
   &NiagaraSpriteRendererProperties
   &NiagaraLightRendererProperties
   &NiagaraMeshRendererProperties
   &NiagaraDecalRendererProperties
  ]
  renderers [
   {rendererClass=&NiagaraSpriteRendererProperties}
   {rendererClass=&NiagaraLightRendererProperties, rendererIndex=1}
   {rendererClass=&NiagaraMeshRendererProperties, rendererIndex=2}
   {rendererClass=&NiagaraMeshRendererProperties, rendererIndex=3}
   {rendererClass=&NiagaraDecalRendererProperties, rendererIndex=4}
   {rendererClass=&NiagaraMeshRendererProperties, rendererIndex=5}
   {rendererClass=&NiagaraSpriteRendererProperties, rendererIndex=6}
  ]
 }
 Shockwave {
  emitterUpdateScript [
   &EmitterState(Life Cycle Mode=::NewEnumerator1, Loop Behavior=::NewEnumerator1, Loop Delay=0.1, Loop Duration=0.35, UseLoopDelay=true)
   &SpawnBurst_Instantaneous(Spawn Group=1, Use Spawn Probability=false, UseLoopCountLimit=false) {module=SpawnBurst_Instantaneous001}
   &SpawnBurst_Instantaneous(Use Spawn Probability=false, UseLoopCountLimit=false)
  ]
  particleSpawnScript [
   &InitializeParticle(
    Color=&MultiplyLinearColor(
     Linear Color=&FNTimeOfDayColor(
      Night Color=<rgba 0.5122 0.5384 0.8333 1>
      Morning Color=<rgba 0.6242 0.7654 0.9219 1>
      Day Color=<rgba 1 0.7073 0.5207 1>
      Evening Color=<rgba 0.6242 0.7654 0.9219 1>
     )
     Scale Factor=&MakeVector4_FromLinearColor(LinearColor=<rgba 1 0.8 0.8 0.5>)
    )
    Lifetime=0.35
    Mesh Scale Mode=::NewEnumerator3
    Mesh Uniform Scale=2.4
    Position Offset=<0 0 7>
    Position Offset Coordinate Space=::Local
    UsePositionOffset=true
    Write Material Random=true
   )
   &CameraOffset()
   /SetVariables(Particles.VisibilityTag=&SpawnGroup()) {isSetParameters=true}
   /SetVariables(
    Particles.Dust Color:LinearColor=&MultiplyLinearColor(
     Linear Color=&FNTimeOfDayColor(
      Night Color=<rgba 0.5122 0.5384 0.8333 1>
      Morning Color=<rgba 0.6242 0.7654 0.9219 1>
      Day Color=<rgba 1 0.7073 0.5207 1>
      Evening Color=<rgba 0.6242 0.7654 0.9219 1>
     )
     Scale Factor=&MakeVector4_FromLinearColor(LinearColor=<rgba 2 1.6 1.6 1>)
    )
    Particles.Dust Scale:Vector3f=&VectorFromFloat(Value=2.4)
   ) {isSetParameters=true}
  ]
  particleUpdateScript [
   &ParticleState()
   &BlendSpace("Position Blend (World <-> Local)":NiagaraFloat=&MakeCustomFloatFromBool(
     Bool=&SetBoolByFloatComparison(A=&VectorLength(Vector=$Engine.Owner.Velocity), B=400)
     True Float=1
     False Float=0
     Translation Contribution:NiagaraFloat=1
     Rotation Contribution:NiagaraFloat=&MakeCustomFloatFromBool(
      Bool=&SetBoolByFloatComparison(A=&VectorLength(Vector=$Engine.Owner.Velocity), B=800)
      True Float=1
      False Float=0
     )
    )) {module=BlendSpace002}
   &DynamicMaterialParameters(
    Index 0 Param 1=&FloatFromCurve(FloatCurve=DataInterface() {propertyValues {Curve {keys [<at 0.5 0 RCIM_Cubic>, {arriveTangent=4.1173, interpMode=RCIM_Cubic, leaveTangent=4.1173, tangentMode=RCTM_User, time=1, value=1}]}}})
    Index 0 Param 2=1
    Index 0 Param 3=1
    Module.Index0Float0Write=true
    Module.Index0Float1Write=true
    Module.Index0Float2Write=true
    Module.Index0Float3Write=true
   )
   &ScaleMeshSize(Scale Factor=&VectorFromCurve(VectorCurve=DataInterface() {propertyValues {
       XCurve {keys [
         {arriveTangent=1.3512, interpMode=RCIM_Cubic, leaveTangent=1.3512, tangentMode=RCTM_User, time=0, value=0.2}
         {arriveTangent=0.0224, interpMode=RCIM_Cubic, leaveTangent=0.0224, tangentMode=RCTM_User, time=1, value=1}
        ]}
       YCurve {keys [
         {arriveTangent=1.3512, interpMode=RCIM_Cubic, leaveTangent=1.3512, tangentMode=RCTM_User, time=0, value=0.2}
         {arriveTangent=0.0224, interpMode=RCIM_Cubic, leaveTangent=0.0224, tangentMode=RCTM_User, time=1, value=1}
        ]}
       ZCurve {keys [{tangentMode=RCTM_User, time=0, value=1}, {tangentMode=RCTM_User, time=1, value=1}]}
      }}))
   &ScaleColor(
    Scale Alpha=&FloatFromCurve(FloatCurve=DataInterface() {propertyValues {Curve {keys [
         {arriveTangent=8.2356, interpMode=RCIM_Cubic, leaveTangent=8.2356, tangentMode=RCTM_User, time=0, value=0.5}
         {arriveTangent=-2.6616, interpMode=RCIM_Cubic, leaveTangent=-2.6616, tangentMode=RCTM_User, time=0.2223, value=1}
         {arriveTangent=-0.1769, interpMode=RCIM_Cubic, leaveTangent=-0.1769, tangentMode=RCTM_User, time=1, value=0}
        ]}}})
    ScaleA=true
    ScaleRGB=true
   )
   &SolveForcesAndVelocity(Clamp Velocity=false, Limit Acceleration=false)
   /SetVariables(
    Particles.Dust Color:LinearColor=&MultiplyLinearColor(
     Linear Color=$Particles.Initial.Dust Color
     Scale Factor=&MakeVector4FromVectorAndFloat(
      "Vector (RGB)"=<1 1 1>
      "Float (Alpha)"=&FloatFromCurve(FloatCurve=DataInterface() {propertyValues {Curve {keys [
           {arriveTangent=8.2356, interpMode=RCIM_Cubic, leaveTangent=8.2356, tangentMode=RCTM_User, time=0, value=0.5}
           {arriveTangent=-2.6616, interpMode=RCIM_Cubic, leaveTangent=-2.6616, tangentMode=RCTM_User, time=0.2223, value=1}
           {arriveTangent=-0.1769, interpMode=RCIM_Cubic, leaveTangent=-0.1769, tangentMode=RCTM_User, time=1, value=0}
          ]}}})
     )
    )
    Particles.Dust Scale:Vector3f=&Multiply_Vector(
     A=$Particles.Initial.Dust Scale
     B=&VectorFromCurve(VectorCurve=DataInterface() {propertyValues {
        XCurve {keys [
          {arriveTangent=1.3512, interpMode=RCIM_Cubic, leaveTangent=1.3512, tangentMode=RCTM_User, time=0, value=0.2}
          {arriveTangent=0.0224, interpMode=RCIM_Cubic, leaveTangent=0.0224, tangentMode=RCTM_User, time=1, value=1}
         ]}
        YCurve {keys [
          {arriveTangent=1.3512, interpMode=RCIM_Cubic, leaveTangent=1.3512, tangentMode=RCTM_User, time=0, value=0.2}
          {arriveTangent=0.0224, interpMode=RCIM_Cubic, leaveTangent=0.0224, tangentMode=RCTM_User, time=1, value=1}
         ]}
        ZCurve {keys [{tangentMode=RCTM_User, time=0, value=1}, {tangentMode=RCTM_User, time=1, value=1}]}
       }})
    )
    Particles.Smoke Dynam 1:Vector4f=&MakeVector4(X=-0.0749, Y=0, Z=0, W=0)
   ) {isSetParameters=true}
  ]
  rendererClasses [&NiagaraMeshRendererProperties]
  renderers [
   {rendererClass=&NiagaraMeshRendererProperties, rendererIndex=0}
   {rendererClass=&NiagaraMeshRendererProperties, rendererIndex=1}
  ]
 }
 Smoke {
  emitterUpdateScript [
   &EmitterState()
   &SpawnBurst_Instantaneous(Spawn Count=40, Spawn Group=2, Spawn Time=0.15, Use Spawn Probability=false, UseLoopCountLimit=false) {module=SpawnBurst_Instantaneous002}
   &SpawnBurst_Instantaneous(Spawn Count=10, Spawn Group=1, Spawn Time=0.15, Use Spawn Probability=false, UseLoopCountLimit=false) {module=SpawnBurst_Instantaneous001}
   &SpawnBurst_Instantaneous(Spawn Count=25, Use Spawn Probability=false, UseLoopCountLimit=false)
  ]
  particleSpawnScript [
   &InitializeParticle(
    Color Channel Mode:ENiagaraChannelCorrelation=::NewEnumerator0
    Color Maximum=<rgba 0.35 0.2752 0.2481 0.25>
    Color Minimum=<rgba 0.35 0.2877 0.257 0.2>
    Color Mode=::NewEnumerator2
    Lifetime=&MergeHelper_FloatSwitch(
     0=&RandomRangeFloat(Minimum=0.8, Maximum=1.2)
     1=&RandomRangeFloat(Minimum=0.18, Maximum=0.3)
     2=&RandomRangeFloat(Minimum=0.2, Maximum=0.55)
     3=?
     Index=&SpawnGroup()
     Num=::NewEnumerator1
    )
    Position Offset=&MergeHelper_VectorSwitch(0=<0 0 0>, 1=<0 0 -15>, 2=<0 0 0>, 3=?, Index=&SpawnGroup(), Num=::NewEnumerator1)
    Sprite Rotation Mode=::NewEnumerator2
    Sprite Size Mode=::NewEnumerator1
    Uniform Sprite Size Max=50
    Uniform Sprite Size Min=40
    UsePositionOffset=true
    Write Material Random=true
   ) {module=InitializeParticle001}
   &ShapeLocation(
    Cylinder Height=0
    Cylinder Height Midpoint=0
    Cylinder Radius=10
    MaskSpawnGroup=false
    Offset Coordinate Space=::Local
    Offset Mode=::NewEnumerator0
    Override Seed=false
    Shape Primitive=::NewEnumerator1
    Surface Only=false
    UseOwnerScale=true
   )
   &ShapeLocation(MaskSpawnGroup=true, Override Seed=false, Spawn Group Mask=2, Sphere Radius=15, UseOwnerScale=true) {module=ShapeLocation001}
   &AddVelocity(
    Module.Use Falloff Distance=false
    Origin Offset=&MergeHelper_VectorSwitch(0=<0 0 0>, 1=<0 0 -27.2686>, 2=<0 0 -20.3274>, 3=?, Index=&SpawnGroup(), Num=::NewEnumerator1)
    Override Seed=false
    Velocity Mode=::NewEnumerator1
    Velocity Speed=&MergeHelper_FloatSwitch(
     0=&RandomRangeFloat(Minimum=500, Maximum=700)
     1=&RandomRangeFloat(Minimum=250, Maximum=350)
     2=&RandomRangeFloat(Minimum=200, Maximum=800)
     3=?
     Index=&SpawnGroup()
     Num=::NewEnumerator1
    )
   )
   &SubUVAnimation(End Frame=3, SubUV Animation Mode:ENiagara_SubUVLookupMode=::NewEnumerator2)
   &CameraOffset(Camera Offset Amount=&MergeHelper_FloatSwitch(0=0, 1=10, 2=?, 3=?, Index=&SpawnGroup(), Num=::NewEnumerator0)) {module=CameraOffset001}
   /SetVariables(Particles.VisibilityTag=&SpawnGroup()) {isSetParameters=true}
   /SetVariables(
    Particles.Streak Color:LinearColor=<rgba 1 0.5052 0.7155 0.5>
    Particles.Streak Sprite Size:Vector2f=&UniformRangedVector2D(Minimum=<40 140>, Maximum=<80 120>)
   ) {isSetParameters=true}
   /SetVariables(
    Particles.Mote Color:LinearColor=&RandomRangeLinearColor(Minimum=<rgba 3 1.5156 2.1464 1>, Maximum=<rgba 1 0.3854 0.6466 1>)
    Particles.Mote Sprite Size:Vector2f=&Vector2DFromFloat(Value=&RandomRangeFloat(Minimum=3, Maximum=6))
   ) {isSetParameters=true}
  ]
  particleUpdateScript [
   /SetVariables(Particles.Jitter:NiagaraFloat=&Waveform("Waveform [1]"=::NewEnumerator12, "[1] Frequency"=25, "[1] Amplitude Min/Max"=<0.5 1>)) {isSetParameters=true}
   &ParticleState()
   &BlendSpace("Position Blend (World <-> Local)":NiagaraFloat=&MakeCustomFloatFromBool(
     Bool=&SetBoolByFloatComparison(A=&VectorLength(Vector=$Engine.Owner.Velocity), B=400)
     True Float=1
     False Float=0
     Translation Contribution:NiagaraFloat=1
     Rotation Contribution:NiagaraFloat=1
    )) {module=BlendSpace002}
   &Drag(
    Drag=&MergeHelper_FloatSwitch(
     0=7
     1=7
     2=&FloatFromCurve(
      FloatCurve=DataInterface() {propertyValues {Curve {keys [<at 0 0 RCIM_Cubic>, {arriveTangent=2.8933, interpMode=RCIM_Cubic, leaveTangent=2.8933, tangentMode=RCTM_User, time=1, value=2}]}}}
      Scale Curve=&RandomRangeFloat(Minimum=10, Maximum=15)
     )
     3=?
     Index=&SpawnGroup()
     Num=::NewEnumerator1
    )
    Use Linear Drag=true
    Use Rotational Drag=true
   )
   &SolveForcesAndVelocity(Clamp Velocity=false, Limit Acceleration=false)
   &ScaleSpriteSize(Uniform Curve Sprite Scale=DataInterface() {propertyValues {Curve {keys [
        {arriveTangent=1.491, interpMode=RCIM_Cubic, leaveTangent=1.491, tangentMode=RCTM_User, time=0, value=0.5}
        {arriveTangent=0.5, interpMode=RCIM_Cubic, leaveTangent=0.5, time=0.95, value=3}
        <at 1 1 RCIM_Cubic>
       ]}}})
   &DynamicMaterialParameters(
    Index 0 Param 1=&Multiply_Float(A=50, B=1)
    Index 1 Param 1=&FloatFromCurve(
     FloatCurve=DataInterface() {propertyValues {Curve {keys [
         {interpMode=RCIM_Cubic, tangentMode=RCTM_User, time=0, value=0}
         {arriveTangent=0.6295, interpMode=RCIM_Cubic, leaveTangent=0.6295, tangentMode=RCTM_User, time=1, value=1}
        ]}}}
     Scale Curve=0.2
    )
    Index 1 Param 2=&FloatFromCurve(
     FloatCurve=DataInterface() {propertyValues {Curve {keys [
         {arriveTangent=-100.8588, time=0, value=1}
         {arriveTangent=-26.7969, time=0.2007, value=0.2}
         {arriveTangent=-2.6967, leaveTangent=-2.6967, time=0.3, value=0.1}
        ]}}}
     CurveIndex=$Emitter.Age
     Scale Curve=&MaskFloatBySpawnGroup(Masked Float=0, Passthrough Float=5, Spawn Group Mask=1)
    )
    Index 1 Param 3=&FloatFromCurve(
     FloatCurve=DataInterface() {propertyValues {Curve {keys [{interpMode=RCIM_Cubic, tangentMode=RCTM_User, time=0.2, value=0.5}, {interpMode=RCIM_Cubic, tangentMode=RCTM_User, time=1, value=1}]}}}
     Scale Curve=2
    )
    Index 1 Param 4=&FloatFromCurve(FloatCurve=DataInterface() {propertyValues {Curve {keys [<at 0 0.3>,<at 1 0>]}}}, Scale Curve=-0.3)
    Module.Index0Float0Write=true
    Module.Index0Float1Write=true
    Module.Index0Float2Write=true
    Module.Index0Float3Write=true
    Module.Index1Float0Write=false
    Module.Index1Float1Write=false
    Module.Index1Float2Write=true
    Module.Index1Float3Write=false
    Write Parameter Index 1=::NewEnumerator1
   )
   &CameraOffset(Camera Offset Amount=5)
   &ScaleColor(
    Scale Alpha=&Multiply_Float(
     A=&FloatFromCurve(
      FloatCurve=DataInterface() {propertyValues {Curve {keys [{interpMode=RCIM_Cubic, tangentMode=RCTM_User, time=0, value=0}, <at 0.25 1 RCIM_Cubic>, <at 0.9 0>]}}}
      Scale Curve=&FN_FrontEndFloatModifier(Gameplay=1, Front End=1)
     )
     B=&RandomRangeFloat(Minimum=0.5, Maximum=1)
    )
    Scale RGB=<5 5 5>
    ScaleA=true
    ScaleRGB=true
   )
   /SetVariables(
    Particles.Smoke Dynam 1:Vector4f=&MakeVector4(
     X=&Add_Float(
      A=&RandomRangeFloat(Minimum=0, Maximum=1)
      B=&FloatFromCurve(
       FloatCurve=DataInterface() {propertyValues {Curve {keys [
           <at 0 0 RCIM_Cubic>
           {arriveTangent=1.5947, interpMode=RCIM_Cubic, leaveTangent=1.5947, tangentMode=RCTM_User, time=0.2557, value=0.5257}
           <at 1 1 RCIM_Cubic>
          ]}}}
       CurveIndex=$Particles.Age
       Scale Curve=0.25
      )
     )
     Y=0
     Z=1
     W=1
    )
    Particles.Smoke Dynam 2:Vector4f=&MakeVector4(X=0, Y=0, Z=0, W=0)
    Particles.Smoke Dynam 3:Vector4f=&MakeVector4(X=0, Y=0, Z=2, W=0)
    Particles.Smoke Dynam 4:Vector4f=&MakeVector4(X=1, Y=70, Z=0, W=0)
   ) {isSetParameters=true}
   /SetVariables(
    Particles.Streak Color:LinearColor=&MultiplyLinearColor(
     Linear Color=$Particles.Initial.Streak Color
     Scale Factor=&Vector4FromCurve(Vector4Curve=DataInterface() {propertyValues {
        AlphaCurve {keys [<at 0 1>,<at 0.4257 1>,<at 1 0>]}
        BlueCurve {keys [<at 0 2>,<at 0.4272 1>]}
        GreenCurve {keys [<at 0 2>,<at 0.4272 1>]}
        RedCurve {keys [<at 0 2>,<at 0.4272 1>]}
       }})
    )
    Particles.Streak Sprite Size:Vector2f=&Multiply_Vector2D(
     A=$Particles.Initial.Streak Sprite Size
     B=&Vector2DFromCurve(
      Vector2Curve=DataInterface() {propertyValues {
        XCurve {keys [{arriveTangent=2.307, interpMode=RCIM_Cubic, leaveTangent=2.307, tangentMode=RCTM_User, time=0, value=0.4}, <at 1 1 RCIM_Cubic>]}
        YCurve {keys [{arriveTangent=3.3653, interpMode=RCIM_Cubic, leaveTangent=3.3653, tangentMode=RCTM_User, time=0, value=0.4}, <at 1 1.5 RCIM_Cubic>]}
       }}
      ScaleCurve=<1 1>
     )
    )
   ) {isSetParameters=true}
   /SetVariables(
    Particles.Mote Color:LinearColor=&MultiplyLinearColor(
     Linear Color=$Particles.Initial.Mote Color
     Scale Factor=&Vector4FromCurve(Vector4Curve=DataInterface() {propertyValues {AlphaCurve {keys [<at 0.4257 1>,<at 1 0>]}, BlueCurve {keys [<at 0 1>]}, GreenCurve {keys [<at 0 1>]}, RedCurve {keys [<at 0 1>]}}})
    )
    Particles.Mote Sprite Size:Vector2f=&Multiply_Vector2D(
     A=$Particles.Initial.Mote Sprite Size
     B=&Vector2DFromCurve(
      Vector2Curve=DataInterface() {propertyValues {
        XCurve {keys [
          {arriveTangent=0.1643, interpMode=RCIM_Cubic, leaveTangent=0.1643, leaveTangentWeight=0.7988, tangentMode=RCTM_User, time=0, value=0.4}
          <at 1 1 RCIM_Cubic>
         ]}
        YCurve {keys [
          {arriveTangent=34.3471, interpMode=RCIM_Cubic, leaveTangent=34.3471, leaveTangentWeight=1.1393, tangentMode=RCTM_User, time=0, value=0.4}
          {arriveTangent=-5.7403, interpMode=RCIM_Cubic, leaveTangent=-5.7403, tangentMode=RCTM_User, time=0.794, value=2}
          <at 1 1.5 RCIM_Cubic>
         ]}
       }}
      ScaleCurve=<1 1.5>
     )
    )
   ) {isSetParameters=true}
  ]
  rendererClasses [&NiagaraSpriteRendererProperties]
  renderers [
   {rendererClass=&NiagaraSpriteRendererProperties}
   {rendererClass=&NiagaraSpriteRendererProperties, rendererIndex=1}
   {rendererClass=&NiagaraSpriteRendererProperties, rendererIndex=2}
   {rendererClass=&NiagaraSpriteRendererProperties, rendererIndex=3}
  ]
 }
}
summary {
 systemName=NS_Emote_SkiffGrind_SpawnIn
 emitters [
  {emitterName=Shockwave, rendererClasses [&NiagaraMeshRendererProperties]}
  {emitterName=Smoke, rendererClasses [&NiagaraSpriteRendererProperties]}
  {emitterName=AirLines, rendererClasses [&NiagaraSpriteRendererProperties, &NiagaraMeshRendererProperties]}
  {emitterName=Decal, rendererClasses [&NiagaraDecalRendererProperties]}
  {
   emitterName=ImpactFlare
   rendererClasses [
    &NiagaraSpriteRendererProperties
    &NiagaraLightRendererProperties
    &NiagaraMeshRendererProperties
    &NiagaraDecalRendererProperties
   ]
  }
 ]
 userVariables [
  User.bIsFrontEnd {type=NiagaraBool}
  User.bIsFrontEndPreview {type=NiagaraBool}
  User.DustSpawnCurve {type=NiagaraFloat, defaultValue=1, description=""}
 ]
}
