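# Niagara system for the SkiffGrind emote gadget's thruster effects, with emitters for back and side thrusters, a defense matrix and sparks.
# Thruster particles follow left and right engine bones, while four staggered, pulsing matrix layers follow a spine bone.
# Spawn groups split each emitter into layers across mesh and sprite renderers via VisibilityTag.
# System-age curves time scale and pulses to the roughly 4.6-second emote, and thruster spawn rates vary by quality level.
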NiagaraSystem /BRCosmetics/Animation/Game/MainPlayer/Emotes/SkiffGrind/FX/NS_Emote_SkiffGrind_Thrusters
dependencies {
 usedDataInterfaces [&NiagaraDataInterfaceColorCurve, &NiagaraDataInterfaceVectorCurve, &NiagaraDataInterfaceCurve]
 usedDynamicInputs [&Waveform]
 usedModules [
  &SystemState
  &BlendSpaceHelper
  &self:SystemSpawnScript.SystemScriptSource.SystemScriptGraph.NiagaraNodeAssignment_0.SetVariables_#1
  &self:SystemSpawnScript.SystemScriptSource.SystemScriptGraph.NiagaraNodeAssignment_1.SetVariables_#2
 ]
}
emitters {
 BackThrusters {
  emitterUpdateScript [
   &EmitterState()
   &SpawnBurst_Instantaneous(Spawn Count=2, Spawn Group=3, Use Spawn Probability=false, UseLoopCountLimit=false) {module=SpawnBurst_Instantaneous002}
   &SpawnBurst_Instantaneous(Spawn Count=4, Spawn Group=2, Use Spawn Probability=false, UseLoopCountLimit=false) {module=SpawnBurst_Instantaneous001}
   &SpawnBurst_Instantaneous(Spawn Count=2, Spawn Group=1, Use Spawn Probability=false, UseLoopCountLimit=false)
   &SpawnRate(
    SpawnRate=&MakeCustomFloatFromBool(
     Bool=&SetBoolByIntComparison(A=$Engine.QualityLevel, B=0, Comparison Type=::NewEnumerator4)
     True Float=15
     False Float=25
    )
    Use Spawn Probability=false
   )
   /SetVariables(Emitter.Inner Thruster Color A:LinearColor=<rgba 0.1066 1 0.3289 1>) {isSetParameters=true, module=SetVariables_#3}
  ]
  particleSpawnScript [
   &PartitionParticles()
   &InitializeParticle(
    Color Mode=::NewEnumerator0
    Lifetime=&MergeHelper_FloatSwitch(
     0=&RandomRangeFloat(Minimum=0.1, Maximum=0.15)
     1=4.592
     2=0.3
     3=4.592
     Index=&SpawnGroup()
     Num=::NewEnumerator2
    )
    Mesh Scale Mode=::NewEnumerator3
    Mesh Uniform Scale=0.1
    UsePositionOffset=false
    Write Material Random=true
   )
   &InitialMeshOrientation(
    Mesh Axis To Orient=&MergeHelper_VectorSwitch(
     0=&MergeHelper_VectorSwitch(0=<1 0 0>, 1=<-1 0 0>, 2=?, 3=?, Index=$Particles.PartitionParticles.Partition, Num=::NewEnumerator0)
     1=&MergeHelper_VectorSwitch(0=<0 0 -1>, 1=<0 0 1>, 2=?, 3=?, Index=$Particles.PartitionParticles.Partition, Num=::NewEnumerator0)
     2=&MergeHelper_VectorSwitch(0=<1 0 0>, 1=<-1 0 0>, 2=?, 3=?, Index=$Particles.PartitionParticles.Partition, Num=::NewEnumerator0)
     3=?
     Index=&SpawnGroup()
     Num=::NewEnumerator1
    )
    Mesh Orientation Mode=::NewEnumerator2
    Orientation Coordinate Space=::Local
    Orientation Vector=&MergeHelper_VectorSwitch(0=<1 0 0>, 1=<1 0 0>, 2=<1 0 0>, 3=?, Index=&SpawnGroup(), Num=::NewEnumerator1)
    Rotation=&MergeHelper_VectorSwitch(
     0=&RandomRangeVector(Minimum=<-1 0 0>, Maximum=<1 0 0>)
     1=<0 0 0>
     2=&RandomRangeVector(Minimum=<-1 0 0>, Maximum=<1 0 0>)
     3=?
     Index=&SpawnGroup()
     Num=::NewEnumerator1
    )
    Rotation Coordinate Space=::NewEnumerator0
    Use Rotation Vector=true
   )
   /SetVariables(Particles.VisibilityTag=&SpawnGroup()) {isSetParameters=true, module=SetVariables_#4}
   /SetVariables(
    Particles.Inner Thruster Color:LinearColor=<rgba 5 5 5 0.03>
    Particles.Inner Thruster Scale:Vector3f=&VectorFromFloat(Value=0.06)
   ) {isSetParameters=true, module=SetVariables_#5}
   /SetVariables(
    Particles.Flare Color:LinearColor=<rgba 0 3 1.1152 0.8>
    Particles.Flare Sprite Size:Vector2f=&Vector2DFromFloat(Value=60)
   ) {isSetParameters=true, module=SetVariables_#6}
   /SetVariables(
    Particles.Initial Burst Scale:Vector3f=&VectorFromFloat(Value=0.1)
    Particles.Initial Burst Sprite Color:LinearColor=<rgba 0.2062 1 0 1>
    Particles.Initial Burst Sprite Size:Vector2f=&Vector2DFromFloat(Value=60)
   ) {isSetParameters=true}
   /SetVariables(Particles.Glow Color=<rgba 0 1 0.3018 0.2>) {isSetParameters=true}
   &CameraOffset(Camera Offset Amount=&MergeHelper_FloatSwitch(0=0, 1=0, 2=0, 3=15, Index=&SpawnGroup(), Num=::NewEnumerator2))
  ]
  particleUpdateScript [
   &ParticleState()
   &SkeletalMeshLocation(
    Bone / Socket Index=$Particles.PartitionParticles.Partition
    Bone Sampling Mode=::NewEnumerator4
    MaskSpawnGroup=false
    Sampled Position Offset=&MergeHelper_VectorSwitch(
     0=&MergeHelper_VectorSwitch(0=<-11 0 0>, 1=<11 0 0>, 2=?, 3=?, Index=$Particles.PartitionParticles.Partition, Num=::NewEnumerator0)
     1=&MergeHelper_VectorSwitch(0=<-11 0 0>, 1=<11 0 0>, 2=?, 3=?, Index=$Particles.PartitionParticles.Partition, Num=::NewEnumerator0)
     2=&MergeHelper_VectorSwitch(0=<-11 0 0>, 1=<11 0 0>, 2=?, 3=?, Index=$Particles.PartitionParticles.Partition, Num=::NewEnumerator0)
     3=&MergeHelper_VectorSwitch(0=<-15 0 0>, 1=<15 0 0>, 2=?, 3=?, Index=$Particles.PartitionParticles.Partition, Num=::NewEnumerator0)
     Index=&SpawnGroup()
     Num=::NewEnumerator2
    )
    Skeletal Mesh=DataInterface() {propertyValues {
      DefaultMesh=null
      ExcludeBoneName=root
      OptionalComponentClass=null
      PreviewMesh=/BRCosmetics/Animation/Props/Gadgets_SkiffGrind/Meshes/SkiffGrind_Gadget
      SkinningMode=SkinOnTheFly
      SoftSourceActor=null
      SourceMode=Default
      UvSetIndex=0
      WholeMeshLOD=-1
      bExcludeBone=false
      bReadDeformedGeometry=true
      bRequireCurrentFrameData=true
      ComponentTags []
      FilteredBones [Tokkimech_backengine_l, Tokkimech_backengine_r]
      FilteredSockets []
      MeshUserParameter {parameter {name=None, typeDefHandle {registeredTypeIndex=1}, varData []}}
      SamplingRegions []
     }}
   )
   &ScaleColor(
    Linear Color Curve=DataInterface() {propertyValues {
      AlphaCurve {keys [<at 0 1>,<at 1 0>]}
      BlueCurve {keys [<at 0 1>,<at 1 0.2476>]}
      GreenCurve {keys [<at 0 1>,<at 1 0.2865>]}
      RedCurve {keys [<at 0 1>,<at 1 0>]}
     }}
    Scale Mode=2
   )
   &ScaleMeshSize(Scale Factor=&VectorFromCurve(
     VectorCurve=DataInterface() {propertyValues {XCurve {keys [<at 0 0.6>,<at 1 1>]}, YCurve {keys [<at 0 0.6>,<at 1 1>]}, ZCurve {keys [<at 0 0.6>,<at 1 1>]}}}
     Scale Curve=<0.6 1.1 1.1>
    ))
   &ScaleMeshSize(Scale Factor=&VectorFromFloat(Value=&FloatFromCurve(FloatCurve=DataInterface() {propertyValues {Curve {keys [<at 0 1>,<at 4.3663 1>,<at 4.592 0>]}}}, CurveIndex=$System.Age))) {module=ScaleMeshSize001}
   /SetVariables(
    Particles.Inner Thruster Color:LinearColor=&MultiplyLinearColor(
     Linear Color=$Particles.Initial.Inner Thruster Color
     Scale Factor=&Vector4FromCurve(Vector4Curve=DataInterface() {propertyValues {
        AlphaCurve {keys [<at 0 0>,<at 0.0198 1>,<at 0.9833 1>,<at 1 0>]}
        BlueCurve {keys [<at 0 1>,<at 1 1>]}
        GreenCurve {keys [<at 0 1>,<at 1 1>]}
        RedCurve {keys [<at 0 1>,<at 1 1>]}
       }})
    )
    Particles.Inner Thruster Scale:Vector3f=&Multiply_Vector(
     A=$Particles.Initial.Inner Thruster Scale
     B=&VectorFromFloat(Value=&FloatFromCurve(FloatCurve=DataInterface() {propertyValues {Curve {keys [<at 0 0>,<at 0.1 1>,<at 4.4626 1>,<at 4.592 0>]}}}, CurveIndex=$System.Age))
    )
   ) {isSetParameters=true, module=SetVariables_#7}
   /SetVariables(
    Particles.Flare Color:LinearColor=&MakeLinearColorFromVectorAndFloat(
     "Vector (RGB)"=&MakeVectorFromLinearColorRGB(Color=$Particles.Initial.Flare Color)
     "Float (Alpha)"=&FloatFromCurve(
      FloatCurve=DataInterface() {propertyValues {Curve {keys [
          {arriveTangent=-0.9, time=0, value=0.15}
          <at 0.0622 0.85>
          {arriveTangent=-1.0872, leaveTangent=-1.0872, time=0.1202, value=0.15}
          <at 0.1772 0.9>
          <at 0.238 0.15>
          {arriveTangent=-1.5912, leaveTangent=-1.5912, time=0.3049, value=0.9}
          <at 0.3675 0.15>
          <at 0.4869 0.9>
          <at 0.5465 0.15>
          {arriveTangent=-0.0998, time=0.5737, value=0.85}
          <at 0.6545 0.15>
          <at 0.7224 0.85>
          <at 0.8185 0.15>
          <at 0.8964 0.85>
          <at 0.9864 0.15>
          <at 1.0002 0.15>
         ]}}}
      Scale Curve=&RandomRangeFloat(Minimum=1, Maximum=3)
     )
    )
    Particles.Flare Sprite Size:Vector2f=&Multiply_Vector2D(
     A=$Particles.Initial.Flare Sprite Size
     B=&Vector2DFromFloat(Value=&FloatFromCurve(FloatCurve=DataInterface() {propertyValues {Curve {keys [
           {interpMode=RCIM_Cubic, leaveTangentWeight=0.2638, tangentMode=RCTM_User, time=0, value=1}
           {arriveTangentWeight=0.3731, interpMode=RCIM_Cubic, leaveTangentWeight=0.0983, tangentMode=RCTM_User, time=0.8982, value=1}
           <at 1 0 RCIM_Cubic>
          ]}}}))
    )
   ) {isSetParameters=true, module=SetVariables_#8}
   /SetVariables(
    Particles.Initial Burst Color:LinearColor=&MultiplyLinearColor(
     Linear Color=$Particles.Initial.Initial Burst Color
     Scale Factor=&Vector4FromCurve(Vector4Curve=DataInterface() {propertyValues {
        AlphaCurve {keys [<at 0 1>,<at 1 0>]}
        BlueCurve {keys [<at 0 1>,<at 1 1>]}
        GreenCurve {keys [<at 0 1>,<at 1 1>]}
        RedCurve {keys [<at 0 1>,<at 1 1>]}
       }})
    )
    Particles.Initial Burst Scale:Vector3f=&Multiply_Vector(
     A=$Particles.Initial.Initial Burst Scale
     B=&VectorFromCurve(
      VectorCurve=DataInterface() {propertyValues {
        XCurve {keys [<at 0 0 RCIM_Cubic>,<at 0.8602 1.3 RCIM_Cubic>,<at 1 0 RCIM_Cubic>]}
        YCurve {keys [
          <at 0 0 RCIM_Cubic>
          {
           arriveTangent=2.9083
           arriveTangentWeight=0.1413
           interpMode=RCIM_Cubic
           leaveTangent=2.9083
           leaveTangentWeight=0.1884
           tangentMode=RCTM_User
           time=0.3
           value=1
          }
          {arriveTangent=-4.5986, interpMode=RCIM_Cubic, leaveTangent=-4.5986, tangentMode=RCTM_User, time=0.7, value=1}
          <at 1 0 RCIM_Cubic>
         ]}
        ZCurve {keys [
          <at 0 0 RCIM_Cubic>
          {
           arriveTangent=2.9083
           arriveTangentWeight=0.1744
           interpMode=RCIM_Cubic
           leaveTangent=2.9083
           leaveTangentWeight=0.2325
           tangentMode=RCTM_User
           time=0.3
           value=1
          }
          {arriveTangent=-4.5986, interpMode=RCIM_Cubic, leaveTangent=-4.5986, tangentMode=RCTM_User, time=0.7, value=1}
          <at 1 0 RCIM_Cubic>
         ]}
       }}
      Scale Curve=<0.6 1.1 1.1>
     )
    )
    Particles.Initial Burst Sprite Color:LinearColor=&MultiplyLinearColor(
     Linear Color=$Particles.Initial.Initial Burst Sprite Color
     Scale Factor=&Vector4FromCurve(Vector4Curve=DataInterface() {propertyValues {
        AlphaCurve {keys [<at 0 1>,<at 1 0>]}
        BlueCurve {keys [<at 0 1>,<at 1 1>]}
        GreenCurve {keys [<at 0 1>,<at 1 1>]}
        RedCurve {keys [<at 0 1>,<at 1 1>]}
       }})
    )
    Particles.Initial Burst Sprite Size:Vector2f=&Multiply_Vector2D(
     A=$Particles.Initial.Initial Burst Sprite Size
     B=&Vector2DFromFloat(Value=&FloatFromCurve(FloatCurve=DataInterface() {propertyValues {Curve {keys [{arriveTangent=2.5025, interpMode=RCIM_Cubic, leaveTangent=2.5025, tangentMode=RCTM_User, time=0, value=0}, <at 1 1 RCIM_Cubic>]}}}))
    )
   ) {isSetParameters=true}
   /SetVariables(Particles.Glow Color=&MultiplyLinearColor(
     Linear Color=$Particles.Initial.Glow Color
     Scale Factor=&Vector4FromCurve(Vector4Curve=DataInterface() {propertyValues {
        AlphaCurve {keys [<at 0 0>,<at 0.0198 1>,<at 0.8646 1>,<at 1 0>]}
        BlueCurve {keys [<at 0 1>,<at 1 1>]}
        GreenCurve {keys [<at 0 1>,<at 1 1>]}
        RedCurve {keys [<at 0 1>,<at 1 1>]}
       }})
    )) {isSetParameters=true}
  ]
  rendererClasses [&NiagaraMeshRendererProperties, &NiagaraSpriteRendererProperties]
  renderers [
   {rendererClass=&NiagaraMeshRendererProperties, rendererIndex=0}
   {rendererClass=&NiagaraMeshRendererProperties, rendererIndex=1}
   {rendererClass=&NiagaraSpriteRendererProperties, rendererIndex=2}
   {rendererClass=&NiagaraMeshRendererProperties, rendererIndex=3}
   {rendererClass=&NiagaraSpriteRendererProperties, rendererIndex=4}
   {rendererClass=&NiagaraSpriteRendererProperties, rendererIndex=5}
   {rendererClass=&NiagaraMeshRendererProperties, rendererIndex=6}
  ]
 }
 DefenseMatrix {
  emitterUpdateScript [
   &EmitterState(Life Cycle Mode=::NewEnumerator1, Loop Delay=0.2, Loop Duration=4.592, UseLoopDelay=true)
   &SpawnRate(SpawnGroup=1, SpawnRate=1.5, Use Spawn Probability=false)
   &SpawnBurst_Instantaneous(Spawn Count=4, Spawn Group=4, Use Spawn Probability=false, UseLoopCountLimit=false) {module=SpawnBurst_Instantaneous003}
   &SpawnBurst_Instantaneous(Spawn Group=3, Use Spawn Probability=false, UseLoopCountLimit=false) {module=SpawnBurst_Instantaneous002}
   &SpawnBurst_Instantaneous(Spawn Group=2, Use Spawn Probability=false, UseLoopCountLimit=false) {module=SpawnBurst_Instantaneous001}
   &SpawnBurst_Instantaneous(Spawn Count=4, Use Spawn Probability=false, UseLoopCountLimit=false)
  ]
  particleSpawnScript [
   &PartitionParticles(
    Number Of Partitions=&MergeHelper_IntSwitch(0=4, 1=1, 2=4, 3=4, 4=4, 5=?, Index=&SpawnGroup(), Num=::NewEnumerator3)
    Particles Per Partition=1
    Partition Mode:ENiagara_PartitionMode_v2=::NewEnumerator1
   )
   &InitializeParticle(
    Color=<rgba 1.1987 4.9045 5 1>
    Lifetime=&MergeHelper_FloatSwitch(0=4.592, 1=1.02, 2=?, 3=?, Index=&SpawnGroup(), Num=::NewEnumerator0)
    Mesh Scale=<1.1621 1.5 1.9171>
    Mesh Scale Mode=::NewEnumerator4
    Sprite Size=&MergeHelper_Vector2Switch(0=<34.1427 20>, 1=<48.4773 27>, 2=<61.2378 32>, 3=<72 40>, 4=?, Index=$Particles.PartitionParticles.Partition, Num=::NewEnumerator2)
    Sprite Size Mode=::NewEnumerator4
    UsePositionOffset=false
    Write Material Random=true
   )
   &SpriteFacingAndAlignment(Sprite Alignment=<0 0 1>, Sprite Facing=<0 1 0>, WriteAlignment=true, WriteFacing=true)
   /SetVariables(Particles.VisibilityTag=&SpawnGroup()) {isSetParameters=true}
   /SetVariables(Particles.Light Color:LinearColor=<rgba 0.3926 1.5 1.3418 1>) {isSetParameters=true}
   /SetVariables(
    Particles.Flare Color:LinearColor=<rgba 0.1626 1 0.8365 1>
    Particles.Flare Sprite Size:Vector2f=&Vector2DFromFloat(Value=60)
   ) {isSetParameters=true}
   &CameraOffset(Camera Offset Amount=&MergeHelper_FloatSwitch(0=0, 1=0, 2=0, 3=25, Index=&SpawnGroup(), Num=::NewEnumerator2))
   /SetVariables(Particles.Glow Color=<rgba 0.3801 2 1.454 1>) {isSetParameters=true}
   /SetVariables(
    Particles.Matrix Mesh Color:LinearColor=<rgba 0.1901 1 0.727 1>
    Particles.Matrix Mesh Scale:Vector3f=&MergeHelper_VectorSwitch(
     0=<0.5 0.6789 0.4082>
     1=<0.6 0.9261 0.5513>
     2=<0.7 1.2276 0.6431>
     3=<0.75 1.4223 0.8291>
     Index=$Particles.PartitionParticles.Partition
     Num=::NewEnumerator2
    )
   ) {isSetParameters=true}
  ]
  particleUpdateScript [
   &SolveForcesAndVelocity(Clamp Velocity=false, Limit Acceleration=false)
   &ParticleState()
   &ScaleMeshSize(Scale Factor=&VectorFromCurve(
     VectorCurve=DataInterface() {propertyValues {
       XCurve {keys [<at 0.2 0>,<at 0.347 1.0004>,<at 4.36 1>,<at 4.592 0>,<at 4.592 0>]}
       YCurve {keys [<at 0.2 0>,<at 0.347 1>,<at 4.36 1>,<at 4.59 0>]}
       ZCurve {keys [<at 0.2 0>,<at 0.347 1.0004>,<at 4.36 1>,<at 4.6227 0>]}
      }}
     CurveIndex=$System.Age
    ))
   &ScaleSpriteSize(
    Uniform Curve Index=$System.Age
    Uniform Curve Sprite Scale=DataInterface() {propertyValues {Curve {keys [
        <at 0.2 0 RCIM_Cubic>
        {
         arriveTangent=1.493
         arriveTangentWeight=0.0682
         interpMode=RCIM_Cubic
         leaveTangent=1.493
         leaveTangentWeight=1.4983
         tangentMode=RCTM_User
         time=0.316
         value=1
        }
        {arriveTangentWeight=0.1483, interpMode=RCIM_Cubic, leaveTangentWeight=1.2571, tangentMode=RCTM_User, time=0.6049, value=1}
        {arriveTangentWeight=1.416, interpMode=RCIM_Cubic, leaveTangentWeight=0.0854, tangentMode=RCTM_User, time=4.35, value=1}
        <at 4.592 0 RCIM_Cubic>
       ]}}}
   )
   &SkeletalMeshLocation(
    Bone Sampling Mode=::NewEnumerator4
    MaskSpawnGroup=false
    Sampled Position Offset=&MergeHelper_VectorSwitch(
     0=&MergeHelper_VectorSwitch(
      0=&Lerp_Vector(
       A=<15 0 0>
       B=<15 -40 0>
       Alpha=&FloatFromCurve(FloatCurve=DataInterface() {propertyValues {Curve {keys [<at 0 0>,<at 0.025 1>,<at 0.925 1>,<at 0.95 0>]}}})
      )
      1=&Lerp_Vector(
       A=<15 0 0>
       B=<15 -70 0>
       Alpha=&FloatFromCurve(FloatCurve=DataInterface() {propertyValues {Curve {keys [<at 0 0>,<at 0.025 1>,<at 0.925 1>,<at 0.95 0>]}}})
      )
      2=&Lerp_Vector(
       A=<15 0 0>
       B=<15 -100 0>
       Alpha=&FloatFromCurve(FloatCurve=DataInterface() {propertyValues {Curve {keys [<at 0 0>,<at 0.025 1>,<at 0.925 1>,<at 0.95 0>]}}})
      )
      3=&Lerp_Vector(
       A=<15 0 0>
       B=<15 -130 0>
       Alpha=&FloatFromCurve(FloatCurve=DataInterface() {propertyValues {Curve {keys [<at 0 0>,<at 0.025 1>,<at 0.925 1>,<at 0.95 0>]}}})
      )
      Index=$Particles.PartitionParticles.Partition
      Num=::NewEnumerator2
     )
     1=&Lerp_Vector(
      A=<15 45 0>
      B=<15 -130 0>
      Alpha=&FloatFromCurve(FloatCurve=DataInterface() {propertyValues {Curve {keys [<at 0 0>,<at 1 1>]}}})
     )
     2=<15 -20 0>
     3=<15.5816 -22.9878 0>
     Index=&SpawnGroup()
     Num=::NewEnumerator2
    )
    Skeletal Mesh=DataInterface() {propertyValues {
      DefaultMesh=null
      ExcludeBoneName=root
      OptionalComponentClass=null
      PreviewMesh=&SkiffGrind_Gadget
      SkinningMode=SkinOnTheFly
      SoftSourceActor=null
      SourceMode=Default
      UvSetIndex=0
      WholeMeshLOD=-1
      bExcludeBone=false
      bReadDeformedGeometry=true
      bRequireCurrentFrameData=true
      ComponentTags []
      FilteredBones [Tokkimech_spine_02]
      FilteredSockets []
      MeshUserParameter {parameter {name=None, typeDefHandle {registeredTypeIndex=1}, varData []}}
      SamplingRegions []
     }}
   )
   /SetVariables(
    Particles.Flare Color:LinearColor=&MultiplyLinearColor(
     Linear Color=$Particles.Initial.Flare Color
     Scale Factor=&Vector4FromCurve(Vector4Curve=DataInterface() {propertyValues {
        AlphaCurve {keys [<at 0 0>,<at 0.0575 0.8916>,<at 0.9041 1>,<at 1 0>]}
        BlueCurve {keys [<at 0 1>]}
        GreenCurve {keys [<at 0 1>]}
        RedCurve {keys [<at 0 1>]}
       }})
    )
    Particles.Flare Sprite Size:Vector2f=&Multiply_Vector2D(A=$Particles.Initial.Flare Sprite Size, B=<1 1>)
   ) {isSetParameters=true}
   &DynamicMaterialParameters(
    Index 0 Param 1=1
    Index 0 Param 2=10
    Index 0 Param 3=0.01
    Index 1 Param 1=20
    Module.Index0Float0Write=true
    Module.Index0Float1Write=true
    Module.Index0Float2Write=true
    Module.Index0Float3Write=true
    Module.Index1Float0Write=true
    Module.Index1Float1Write=true
    Module.Index1Float2Write=true
    Module.Index1Float3Write=true
    Write Parameter Index 1=::NewEnumerator1
   )
   &ScaleColor(
    Scale Alpha=&MergeHelper_FloatSwitch(
     0=&MergeHelper_FloatSwitch(
      0=&FloatFromCurve(
       FloatCurve=DataInterface() {propertyValues {Curve {keys [<at 0 0.5>,<at 1.3433 0.5>,<at 1.4445 25>,<at 1.6061 0.5>,<at 1.9636 0.5>,<at 2.0118 25>,<at 2.2278 0.5>,<at 2.647 0.5>,<at 2.6807 25>,<at 2.9116 0.5>,<at 3.3327 0.5>,<at 3.3758 25>,<at 3.5772 0.5>,<at 4.0058 0.5>,<at 4.0589 25>,<at 4.2224 0.5>]}}}
       CurveIndex=$System.Age
      )
      1=&FloatFromCurve(
       FloatCurve=DataInterface() {propertyValues {Curve {keys [<at 0.2 0.4>,<at 1.5033 0.4>,<at 1.6045 20>,<at 1.7661 0.4>,<at 2.1236 0.4>,<at 2.1718 20>,<at 2.3878 0.4>,<at 2.807 0.4>,<at 2.8407 20>,<at 3.0716 0.4>,<at 3.4927 0.4>,<at 3.5358 20>,<at 3.7372 0.4>,<at 4.1658 0.4>,<at 4.2189 20>,<at 4.3824 0.4>]}}}
       CurveIndex=$System.Age
      )
      2=&FloatFromCurve(
       FloatCurve=DataInterface() {propertyValues {Curve {keys [<at 0.2 0.3>,<at 1.6633 0.3>,<at 1.7645 15>,<at 1.9261 0.3>,<at 2.2836 0.3>,<at 2.3318 15>,<at 2.5478 0.3>,<at 2.967 0.3>,<at 3.0007 15>,<at 3.2316 0.3>,<at 3.6527 0.3>,<at 3.6958 15>,<at 3.8972 0.3>,<at 4.3258 0.3>,<at 4.3789 15>,<at 4.5424 0.3>]}}}
       CurveIndex=$System.Age
      )
      3=&FloatFromCurve(
       FloatCurve=DataInterface() {propertyValues {Curve {keys [<at 0.2 0.2>,<at 1.8233 0.2>,<at 1.9245 10>,<at 2.0861 0.2>,<at 2.4436 0.2>,<at 2.4918 10>,<at 2.7078 0.2>,<at 3.127 0.2>,<at 3.1607 10>,<at 3.3916 0.2>,<at 3.8127 0.2>,<at 3.8558 10>,<at 4.0572 0.2>,<at 4.4858 0.2>,<at 4.5389 10>,<at 4.7024 0.2>]}}}
       CurveIndex=$System.Age
      )
      Index=$Particles.PartitionParticles.Partition
      Num=::NewEnumerator2
     )
     1=1
     2=1
     3=1
     Index=&SpawnGroup()
     Num=::NewEnumerator2
    )
    ScaleA=true
    ScaleRGB=true
   )
   /SetVariables(Particles.Glow Color=&MultiplyLinearColor(
     Linear Color=$Particles.Initial.Glow Color
     Scale Factor=&MakeVector4FromVectorAndFloat(
      "Vector (RGB)"=<1 1 1>
      "Float (Alpha)"=&MergeHelper_FloatSwitch(
       0=&FloatFromCurve(
        FloatCurve=DataInterface() {propertyValues {Curve {keys [<at 0 1>,<at 1.3433 1>,<at 1.4445 2>,<at 1.6061 1>,<at 1.9636 1>,<at 2.0118 2>,<at 2.2278 1>,<at 2.647 1>,<at 2.6807 2>,<at 2.9116 1>,<at 3.3327 1>,<at 3.3758 2>,<at 3.5772 1>,<at 4.0058 1>,<at 4.0589 2>,<at 4.2224 1>]}}}
        CurveIndex=$System.Age
       )
       1=&FloatFromCurve(
        FloatCurve=DataInterface() {propertyValues {Curve {keys [<at 0.2 0.8>,<at 1.5033 0.8>,<at 1.6045 2>,<at 1.7661 0.8>,<at 2.1236 0.8>,<at 2.1718 2>,<at 2.3878 0.8>,<at 2.807 0.8>,<at 2.8407 2>,<at 3.0716 0.8>,<at 3.4927 0.8>,<at 3.5358 2>,<at 3.7372 0.8>,<at 4.1658 0.8>,<at 4.2189 2>,<at 4.3824 0.8>]}}}
        CurveIndex=$System.Age
       )
       2=&FloatFromCurve(
        FloatCurve=DataInterface() {propertyValues {Curve {keys [<at 0.2 0.6>,<at 1.6633 0.6>,<at 1.7645 2>,<at 1.9261 0.6>,<at 2.2836 0.6>,<at 2.3318 2>,<at 2.5478 0.6>,<at 2.967 0.6>,<at 3.0007 2>,<at 3.2316 0.6>,<at 3.6527 0.6>,<at 3.6958 2>,<at 3.8972 0.6>,<at 4.3258 0.6>,<at 4.3789 2>,<at 4.5424 0.6>]}}}
        CurveIndex=$System.Age
       )
       3=&FloatFromCurve(
        FloatCurve=DataInterface() {propertyValues {Curve {keys [<at 0.2 0.4>,<at 1.8233 0.4>,<at 1.9245 2>,<at 2.0861 0.4>,<at 2.4436 0.4>,<at 2.4918 2>,<at 2.7078 0.4>,<at 3.127 0.4>,<at 3.1607 2>,<at 3.3916 0.4>,<at 3.8127 0.4>,<at 3.8558 2>,<at 4.0572 0.4>,<at 4.4858 0.4>,<at 4.5389 2>,<at 4.7024 0.4>]}}}
        CurveIndex=$System.Age
       )
       Index=$Particles.PartitionParticles.Partition
       Num=::NewEnumerator2
      )
     )
    )) {isSetParameters=true}
   /SetVariables(
    Particles.Matrix Mesh Color:LinearColor=&MultiplyLinearColor(
     Linear Color=$Particles.Initial.Matrix Mesh Color
     Scale Factor=&MakeVector4FromVectorAndFloat(
      "Vector (RGB)"=<1 1 1>
      "Float (Alpha)"=&MergeHelper_FloatSwitch(
       0=&MergeHelper_FloatSwitch(
        0=&FloatFromCurve(
         FloatCurve=DataInterface() {propertyValues {Curve {keys [<at 0 0.5>,<at 1.3433 0.5>,<at 1.4445 1>,<at 1.6061 0.5>,<at 1.9636 0.5>,<at 2.0118 1>,<at 2.2278 0.5>,<at 2.647 0.5>,<at 2.6807 1>,<at 2.9116 0.5>,<at 3.3327 0.5>,<at 3.3758 1>,<at 3.5772 0.5>,<at 4.0058 0.5>,<at 4.0589 1>,<at 4.2224 0.5>]}}}
         CurveIndex=$System.Age
        )
        1=&FloatFromCurve(
         FloatCurve=DataInterface() {propertyValues {Curve {keys [<at 0.2 0.4>,<at 1.5033 0.4>,<at 1.6045 1>,<at 1.7661 0.4>,<at 2.1236 0.4>,<at 2.1718 1>,<at 2.3878 0.4>,<at 2.807 0.4>,<at 2.8407 1>,<at 3.0716 0.4>,<at 3.4927 0.4>,<at 3.5358 1>,<at 3.7372 0.4>,<at 4.1658 0.4>,<at 4.2189 1>,<at 4.3824 0.4>]}}}
         CurveIndex=$System.Age
        )
        2=&FloatFromCurve(
         FloatCurve=DataInterface() {propertyValues {Curve {keys [<at 0.2 0.3>,<at 1.6633 0.3>,<at 1.7645 1>,<at 1.9261 0.3>,<at 2.2836 0.3>,<at 2.3318 1>,<at 2.5478 0.3>,<at 2.967 0.3>,<at 3.0007 1>,<at 3.2316 0.3>,<at 3.6527 0.3>,<at 3.6958 1>,<at 3.8972 0.3>,<at 4.3258 0.3>,<at 4.3789 1>,<at 4.5424 0.3>]}}}
         CurveIndex=$System.Age
        )
        3=&FloatFromCurve(
         FloatCurve=DataInterface() {propertyValues {Curve {keys [<at 0.2 0.2>,<at 1.8233 0.2>,<at 1.9245 1>,<at 2.0861 0.2>,<at 2.4436 0.2>,<at 2.4918 1>,<at 2.7078 0.2>,<at 3.127 0.2>,<at 3.1607 1>,<at 3.3916 0.2>,<at 3.8127 0.2>,<at 3.8558 1>,<at 4.0572 0.2>,<at 4.4858 0.2>,<at 4.5389 1>,<at 4.7024 0.2>]}}}
         CurveIndex=$System.Age
        )
        Index=$Particles.PartitionParticles.Partition
        Num=::NewEnumerator2
       )
       1=1
       2=1
       3=1
       Index=&SpawnGroup()
       Num=::NewEnumerator2
      )
     )
    )
    Particles.Matrix Mesh Scale:Vector3f=&Multiply_Vector(
     A=$Particles.Initial.Matrix Mesh Scale
     B=&VectorFromFloat(Value=&FloatFromCurve(
       FloatCurve=DataInterface() {propertyValues {Curve {keys [
           <at 0.205 0 RCIM_Cubic>
           {
            arriveTangent=0.1659
            arriveTangentWeight=0.0396
            interpMode=RCIM_Cubic
            leaveTangent=0.1659
            leaveTangentWeight=0.0878
            tangentMode=RCTM_User
            time=0.3196
            value=1
           }
           <at 0.5741 1>
           <at 4.375 1>
           <at 4.5541 0>
          ]}}}
       CurveIndex=$System.Age
      ))
    )
   ) {isSetParameters=true}
   /SetVariables(
    Particles.Matrix Pulse Color:LinearColor=&MultiplyLinearColor(
     Linear Color=$Particles.Initial.Color
     Scale Factor=&MakeVector4FromVectorAndFloat(
      "Vector (RGB)"=<1 1 1>
      "Float (Alpha)"=&Multiply_Float(
       A=&FloatFromCurve(FloatCurve=DataInterface() {propertyValues {Curve {keys [<at 0 0>,<at 0.0916 1>,<at 0.6475 1>,<at 1 0>]}}}, Scale Curve=2)
       B=&FloatFromCurve(
        FloatCurve=DataInterface() {propertyValues {Curve {keys [
            <at 0.2 0 RCIM_Cubic>
            {arriveTangentWeight=0.0695, interpMode=RCIM_Cubic, leaveTangentWeight=0.173, tangentMode=RCTM_User, time=0.316, value=1}
            {arriveTangentWeight=0.0963, interpMode=RCIM_Cubic, leaveTangentWeight=1.1837, tangentMode=RCTM_User, time=0.6049, value=1}
            {arriveTangentWeight=1.416, interpMode=RCIM_Cubic, leaveTangentWeight=0.0854, tangentMode=RCTM_User, time=4.2772, value=0.925}
            <at 4.3875 -0.075 RCIM_Cubic>
           ]}}}
        CurveIndex=$System.Age
       )
      )
     )
    )
    Particles.Matrix Pulse Sprite Size:Vector2f=&Multiply_Vector2D(
     A=<72 40>
     B=&Vector2DFromFloat(Value=&Multiply_Float(
       A=&FloatFromCurve(FloatCurve=DataInterface() {propertyValues {Curve {keys [{tangentMode=RCTM_User, time=0, value=0}, {tangentMode=RCTM_User, time=1, value=1}]}}})
       B=&FloatFromCurve(
        FloatCurve=DataInterface() {propertyValues {Curve {keys [
            <at 0.2 0 RCIM_Cubic>
            {
             arriveTangent=1.493
             arriveTangentWeight=0.0682
             interpMode=RCIM_Cubic
             leaveTangent=1.493
             leaveTangentWeight=1.4983
             tangentMode=RCTM_User
             time=0.316
             value=1
            }
            {arriveTangentWeight=0.1483, interpMode=RCIM_Cubic, leaveTangentWeight=1.2571, tangentMode=RCTM_User, time=0.6049, value=1}
            {arriveTangentWeight=1.416, interpMode=RCIM_Cubic, leaveTangentWeight=0.0854, tangentMode=RCTM_User, time=4.35, value=1}
            <at 4.4909 -0.0413 RCIM_Cubic>
           ]}}}
        CurveIndex=$System.Age
       )
      ))
    )
   ) {isSetParameters=true}
   &AlignSpriteToMeshOrientation(
    Mesh Orientation Relative Sprite Alignment Vector=<1 0 0>
    Mesh Orientation Relative Sprite Facing Vector:Vector3f=<0 1 0>
    Orientation Quaternion:Quat4f=$Particles.SkeletalMeshLocation.SampledRotation
    SetAlignment=true
    SetFacing=true
   )
  ]
  rendererClasses [&NiagaraSpriteRendererProperties, &NiagaraMeshRendererProperties]
  renderers [
   {rendererClass=&NiagaraSpriteRendererProperties}
   {rendererClass=&NiagaraSpriteRendererProperties, rendererIndex=1}
   {rendererClass=&NiagaraMeshRendererProperties, rendererIndex=2}
   {rendererClass=&NiagaraSpriteRendererProperties, rendererIndex=3}
   {rendererClass=&NiagaraMeshRendererProperties, rendererIndex=4}
   {rendererClass=&NiagaraSpriteRendererProperties, rendererIndex=5}
  ]
 }
 SideThrusters {
  emitterUpdateScript [
   &EmitterState()
   &SpawnBurst_Instantaneous(Spawn Count=2, Spawn Group=3, Use Spawn Probability=false, UseLoopCountLimit=false) {module=SpawnBurst_Instantaneous002}
   &SpawnBurst_Instantaneous(Spawn Count=4, Spawn Group=2, Use Spawn Probability=false, UseLoopCountLimit=false) {module=SpawnBurst_Instantaneous001}
   &SpawnBurst_Instantaneous(Spawn Count=2, Spawn Group=1, Spawn Time=0.1, Use Spawn Probability=false, UseLoopCountLimit=false)
   &SpawnRate(
    SpawnRate=&MakeCustomFloatFromBool(
     Bool=&SetBoolByIntComparison(A=$Engine.QualityLevel, B=0, Comparison Type=::NewEnumerator4)
     True Float=18
     False Float=25
    )
    Use Spawn Probability=false
   )
   /SetVariables(Emitter.Inner Thruster Color A:LinearColor=<rgba 0.1066 1 0.3289 1>) {isSetParameters=true, module=SetVariables_#3}
  ]
  particleSpawnScript [
   &PartitionParticles()
   &InitializeParticle(
    Color Mode=::NewEnumerator0
    Lifetime=&MergeHelper_FloatSwitch(
     0=&RandomRangeFloat(Minimum=0.1, Maximum=0.15)
     1=4.592
     2=0.3
     3=4.592
     Index=&SpawnGroup()
     Num=::NewEnumerator2
    )
    Mesh Scale Mode=::NewEnumerator3
    Mesh Uniform Scale=0.1
    Sprite Rotation Mode=::NewEnumerator2
    UsePositionOffset=false
    Write Material Random=true
   )
   &InitialMeshOrientation(
    Mesh Axis To Orient=&MergeHelper_VectorSwitch(
     0=&MergeHelper_VectorSwitch(0=<0 -1 0>, 1=<0 1 0>, 2=?, 3=?, Index=$Particles.PartitionParticles.Partition, Num=::NewEnumerator0)
     1=&MergeHelper_VectorSwitch(0=<0 0 -1>, 1=<0 0 1>, 2=?, 3=?, Index=$Particles.PartitionParticles.Partition, Num=::NewEnumerator0)
     2=&MergeHelper_VectorSwitch(0=<0 -1 0>, 1=<0 1 0>, 2=?, 3=?, Index=$Particles.PartitionParticles.Partition, Num=::NewEnumerator0)
     3=?
     Index=&SpawnGroup()
     Num=::NewEnumerator1
    )
    Mesh Orientation Mode=::NewEnumerator2
    Orientation Coordinate Space=::Local
    Orientation Vector=&MergeHelper_VectorSwitch(0=<1 0 0>, 1=<0 1 0>, 2=<1 0 0>, 3=?, Index=&SpawnGroup(), Num=::NewEnumerator1)
    Rotation=&MergeHelper_VectorSwitch(
     0=&RandomRangeVector(Minimum=<0 -1 0>, Maximum=<0 1 0>)
     1=<0 0 0>
     2=&RandomRangeVector(Minimum=<0 -1 0>, Maximum=<0 1 0>)
     3=?
     Index=&SpawnGroup()
     Num=::NewEnumerator1
    )
    Rotation Coordinate Space=::NewEnumerator0
    Use Rotation Vector=true
   )
   /SetVariables(Particles.VisibilityTag=&SpawnGroup()) {isSetParameters=true, module=SetVariables_#4}
   /SetVariables(
    Particles.Inner Thruster Color:LinearColor=<rgba 5 5 5 0.03>
    Particles.Inner Thruster Scale:Vector3f=&VectorFromFloat(Value=0.1)
   ) {isSetParameters=true, module=SetVariables_#5}
   /SetVariables(
    Particles.Flare Color:LinearColor=<rgba 0 3 1.1152 1>
    Particles.Flare Sprite Size:Vector2f=&Vector2DFromFloat(Value=60)
   ) {isSetParameters=true, module=SetVariables_#6}
   /SetVariables(
    Particles.Initial Burst Scale:Vector3f=&VectorFromFloat(Value=0.1)
    Particles.Initial Burst Sprite Color:LinearColor=<rgba 0.2062 1 0 1>
    Particles.Initial Burst Sprite Size:Vector2f=&Vector2DFromFloat(Value=60)
   ) {isSetParameters=true}
   /SetVariables(Particles.Glow Color=<rgba 0 1 0.3018 0.2>) {isSetParameters=true}
   &CameraOffset(Camera Offset Amount=&MergeHelper_FloatSwitch(0=0, 1=0, 2=0, 3=15, Index=&SpawnGroup(), Num=::NewEnumerator2))
  ]
  particleUpdateScript [
   &ParticleState()
   &SkeletalMeshLocation(
    Bone / Socket Index=$Particles.PartitionParticles.Partition
    Bone Sampling Mode=::NewEnumerator4
    MaskSpawnGroup=false
    Sampled Position Offset=&MergeHelper_VectorSwitch(
     0=&MergeHelper_VectorSwitch(0=<-13 -18 1>, 1=<13 18 -1>, 2=?, 3=?, Index=$Particles.PartitionParticles.Partition, Num=::NewEnumerator0)
     1=&MergeHelper_VectorSwitch(0=<-13 -22 1>, 1=<13 22 -1>, 2=?, 3=?, Index=$Particles.PartitionParticles.Partition, Num=::NewEnumerator0)
     2=&MergeHelper_VectorSwitch(0=<-13 -18 1>, 1=<13 18 -1>, 2=?, 3=?, Index=$Particles.PartitionParticles.Partition, Num=::NewEnumerator0)
     3=&MergeHelper_VectorSwitch(0=<-13 -25 1>, 1=<13 25 -1>, 2=?, 3=?, Index=$Particles.PartitionParticles.Partition, Num=::NewEnumerator0)
     Index=&SpawnGroup()
     Num=::NewEnumerator2
    )
    Skeletal Mesh=DataInterface() {propertyValues {
      DefaultMesh=null
      ExcludeBoneName=root
      OptionalComponentClass=null
      PreviewMesh=&SkiffGrind_Gadget
      SkinningMode=SkinOnTheFly
      SoftSourceActor=null
      SourceMode=Default
      UvSetIndex=0
      WholeMeshLOD=-1
      bExcludeBone=false
      bReadDeformedGeometry=true
      bRequireCurrentFrameData=true
      ComponentTags []
      FilteredBones [Tokkimech_sideengine_l, Tokkimech_sideengine_r]
      FilteredSockets []
      MeshUserParameter {parameter {name=None, typeDefHandle {registeredTypeIndex=1}, varData []}}
      SamplingRegions []
     }}
   )
   &ScaleColor(
    Linear Color Curve=DataInterface() {propertyValues {
      AlphaCurve {keys [<at 0 1>,<at 1 0>]}
      BlueCurve {keys [<at 0 1>,<at 1 0.2476>]}
      GreenCurve {keys [<at 0 1>,<at 1 0.2865>]}
      RedCurve {keys [<at 0 1>,<at 1 0>]}
     }}
    Scale Mode=2
   )
   &ScaleMeshSize(Scale Factor=&VectorFromCurve(VectorCurve=DataInterface() {propertyValues {XCurve {keys [<at 0 0.6>,<at 1 1>]}, YCurve {keys [<at 0 0.6>,<at 1 1>]}, ZCurve {keys [<at 0 0.6>,<at 1 1>]}}}))
   &ScaleMeshSize(Scale Factor=&VectorFromFloat(Value=&FloatFromCurve(FloatCurve=DataInterface() {propertyValues {Curve {keys [<at 0 1>,<at 4.3663 1>,<at 4.592 0>]}}}, CurveIndex=$System.Age))) {module=ScaleMeshSize001}
   /SetVariables(
    Particles.Inner Thruster Color:LinearColor=&MultiplyLinearColor(
     Linear Color=$Particles.Initial.Inner Thruster Color
     Scale Factor=&Vector4FromCurve(Vector4Curve=DataInterface() {propertyValues {
        AlphaCurve {keys [<at 0 0>,<at 0.0198 1>,<at 0.9833 1>,<at 1 0>]}
        BlueCurve {keys [<at 0 1>,<at 1 1>]}
        GreenCurve {keys [<at 0 1>,<at 1 1>]}
        RedCurve {keys [<at 0 1>,<at 1 1>]}
       }})
    )
    Particles.Inner Thruster Scale:Vector3f=&Multiply_Vector(
     A=$Particles.Initial.Inner Thruster Scale
     B=&VectorFromFloat(Value=&FloatFromCurve(FloatCurve=DataInterface() {propertyValues {Curve {keys [<at 0 0>,<at 0.1 1>,<at 4.4626 1>,<at 4.592 0>]}}}, CurveIndex=$System.Age))
    )
   ) {isSetParameters=true, module=SetVariables_#7}
   /SetVariables(
    Particles.Flare Color:LinearColor=&MakeLinearColorFromVectorAndFloat(
     "Vector (RGB)"=&MakeVectorFromLinearColorRGB(Color=$Particles.Initial.Flare Color)
     "Float (Alpha)"=&FloatFromCurve(
      FloatCurve=DataInterface() {propertyValues {Curve {keys [
          {arriveTangent=-0.9, time=0, value=0}
          {arriveTangent=-0.0998, time=0.05, value=1}
          {arriveTangent=-1.0872, leaveTangent=-1.0872, time=0.0688, value=0}
          <at 0.1141 0.95>
          <at 0.1507 0>
          <at 0.1987 1>
          <at 0.2343 0>
          <at 0.277 1>
          <at 0.316 0>
          <at 0.3728 1>
          <at 0.4234 0>
          <at 0.5255 0.95>
          <at 0.5315 0>
          <at 0.6404 1>
          <at 0.6954 0>
          {arriveTangent=-1.5912, leaveTangent=-1.5912, time=0.7682, value=1}
          <at 0.8633 0>
          <at 0.9501 1>
          <at 1.0002 0>
         ]}}}
      Scale Curve=3
     )
    )
    Particles.Flare Sprite Size:Vector2f=&Multiply_Vector2D(
     A=$Particles.Initial.Flare Sprite Size
     B=&Vector2DFromFloat(Value=&FloatFromCurve(FloatCurve=DataInterface() {propertyValues {Curve {keys [
           {interpMode=RCIM_Cubic, leaveTangentWeight=0.2638, tangentMode=RCTM_User, time=0, value=1}
           {arriveTangentWeight=0.3731, interpMode=RCIM_Cubic, leaveTangentWeight=0.0983, tangentMode=RCTM_User, time=0.8982, value=1}
           <at 1 0 RCIM_Cubic>
          ]}}}))
    )
   ) {isSetParameters=true, module=SetVariables_#8}
   /SetVariables(
    Particles.Initial Burst Color:LinearColor=&MultiplyLinearColor(
     Linear Color=$Particles.Initial.Initial Burst Color
     Scale Factor=&Vector4FromCurve(Vector4Curve=DataInterface() {propertyValues {
        AlphaCurve {keys [<at 0 1>,<at 1 0>]}
        BlueCurve {keys [<at 0 1>,<at 1 1>]}
        GreenCurve {keys [<at 0 1>,<at 1 1>]}
        RedCurve {keys [<at 0 1>,<at 1 1>]}
       }})
    )
    Particles.Initial Burst Scale:Vector3f=&Multiply_Vector(
     A=$Particles.Initial.Initial Burst Scale
     B=&VectorFromCurve(VectorCurve=DataInterface() {propertyValues {
        XCurve {keys [<at 0 0 RCIM_Cubic>,<at 0.8602 1.3 RCIM_Cubic>,<at 1 0 RCIM_Cubic>]}
        YCurve {keys [
          <at 0 0 RCIM_Cubic>
          {
           arriveTangent=2.9083
           arriveTangentWeight=0.1413
           interpMode=RCIM_Cubic
           leaveTangent=2.9083
           leaveTangentWeight=0.1884
           tangentMode=RCTM_User
           time=0.3
           value=1
          }
          {arriveTangent=-4.5986, interpMode=RCIM_Cubic, leaveTangent=-4.5986, tangentMode=RCTM_User, time=0.7, value=1}
          <at 1 0 RCIM_Cubic>
         ]}
        ZCurve {keys [
          <at 0 0 RCIM_Cubic>
          {
           arriveTangent=2.9083
           arriveTangentWeight=0.1744
           interpMode=RCIM_Cubic
           leaveTangent=2.9083
           leaveTangentWeight=0.2325
           tangentMode=RCTM_User
           time=0.3
           value=1
          }
          {arriveTangent=-4.5986, interpMode=RCIM_Cubic, leaveTangent=-4.5986, tangentMode=RCTM_User, time=0.7, value=1}
          <at 1 0 RCIM_Cubic>
         ]}
       }})
    )
    Particles.Initial Burst Sprite Color:LinearColor=&MultiplyLinearColor(
     Linear Color=$Particles.Initial.Initial Burst Sprite Color
     Scale Factor=&Vector4FromCurve(Vector4Curve=DataInterface() {propertyValues {
        AlphaCurve {keys [<at 0 1>,<at 1 0>]}
        BlueCurve {keys [<at 0 1>,<at 1 1>]}
        GreenCurve {keys [<at 0 1>,<at 1 1>]}
        RedCurve {keys [<at 0 1>,<at 1 1>]}
       }})
    )
    Particles.Initial Burst Sprite Size:Vector2f=&Multiply_Vector2D(
     A=$Particles.Initial.Initial Burst Sprite Size
     B=&Vector2DFromFloat(Value=&FloatFromCurve(FloatCurve=DataInterface() {propertyValues {Curve {keys [{arriveTangent=2.5025, interpMode=RCIM_Cubic, leaveTangent=2.5025, tangentMode=RCTM_User, time=0, value=0}, <at 1 1 RCIM_Cubic>]}}}))
    )
   ) {isSetParameters=true}
   /SetVariables(Particles.Glow Color=&MultiplyLinearColor(
     Linear Color=$Particles.Initial.Glow Color
     Scale Factor=&Vector4FromCurve(Vector4Curve=DataInterface() {propertyValues {
        AlphaCurve {keys [<at 0 0>,<at 0.0198 1>,<at 0.8646 1>,<at 1 0>]}
        BlueCurve {keys [<at 0 1>,<at 1 1>]}
        GreenCurve {keys [<at 0 1>,<at 1 1>]}
        RedCurve {keys [<at 0 1>,<at 1 1>]}
       }})
    )) {isSetParameters=true}
   &DynamicMaterialParameters(Module.Index0Float0Write=true, Module.Index0Float1Write=true, Module.Index0Float2Write=true, Module.Index0Float3Write=true)
  ]
  rendererClasses [&NiagaraMeshRendererProperties, &NiagaraSpriteRendererProperties]
  renderers [
   {rendererClass=&NiagaraMeshRendererProperties, rendererIndex=0}
   {rendererClass=&NiagaraMeshRendererProperties, rendererIndex=1}
   {rendererClass=&NiagaraSpriteRendererProperties, rendererIndex=2}
   {rendererClass=&NiagaraMeshRendererProperties, rendererIndex=3}
   {rendererClass=&NiagaraSpriteRendererProperties, rendererIndex=4}
   {rendererClass=&NiagaraSpriteRendererProperties, rendererIndex=5}
   {rendererClass=&NiagaraMeshRendererProperties, rendererIndex=6}
  ]
 }
 Sparks {
  emitterUpdateScript [
   &EmitterState(Life Cycle Mode=::NewEnumerator1, Loop Behavior=::NewEnumerator1, Loop Delay=0.15, Loop Duration=4.592, UseLoopDelay=true)
   &SpawnRate(SpawnRate=20, Use Spawn Probability=false)
  ]
  particleSpawnScript [
   &PartitionParticles(Number Of Partitions=4)
   &InitializeParticle(
    Color=<rgba 0.0865 1 0.3622 1>
    Lifetime Max=0.25
    Lifetime Min=0.15
    Lifetime Mode=::NewEnumerator1
    Sprite Size Mode=::NewEnumerator1
    Uniform Sprite Size Max=1.5
    Uniform Sprite Size Min=1
    UsePositionOffset=false
    Write Material Random=true
   )
   &SkeletalMeshLocation(
    Bone / Socket Index=$Particles.PartitionParticles.Partition
    Bone Sampling Mode=::NewEnumerator4
    MaskSpawnGroup=false
    Sampled Position Offset=&MergeHelper_VectorSwitch(0=<-11 0 0>, 1=<11 0 0>, 2=<-13 -20 1>, 3=<13 20 -1>, Index=$Particles.PartitionParticles.Partition, Num=::NewEnumerator2)
    Skeletal Mesh=DataInterface() {propertyValues {
      DefaultMesh=null
      ExcludeBoneName=root
      OptionalComponentClass=null
      PreviewMesh=&SkiffGrind_Gadget
      SkinningMode=SkinOnTheFly
      SoftSourceActor=null
      SourceMode=Default
      UvSetIndex=0
      WholeMeshLOD=-1
      bExcludeBone=false
      bReadDeformedGeometry=true
      bRequireCurrentFrameData=true
      ComponentTags []
      FilteredBones [Tokkimech_backengine_l, Tokkimech_backengine_r, Tokkimech_sideengine_l, Tokkimech_sideengine_r]
      FilteredSockets []
      MeshUserParameter {parameter {name=None, typeDefHandle {registeredTypeIndex=1}, varData []}}
      SamplingRegions []
     }}
   )
   &AddVelocity(Velocity=&RandomRangeVector(Minimum=<0 -200 0>, Maximum=<0 -300 0>))
   &ShapeLocation(MaskSpawnGroup=false, Override Seed=false, Sphere Radius=6, UseOwnerScale=true)
  ]
  particleUpdateScript [
   &ParticleState()
   &BlendSpace("Position Blend (World <-> Local)":NiagaraFloat=&MakeCustomFloatFromBool(
     Bool=&SetBoolByFloatComparison(A=&VectorLength(Vector=$Engine.Owner.Velocity), B=400)
     True Float=1
     False Float=&MergeHelper_FloatSwitch(
      0=&FloatFromCurve(FloatCurve=DataInterface() {propertyValues {Curve {keys [<at 0 1 RCIM_Cubic>, {arriveTangent=-3.5186, interpMode=RCIM_Cubic, leaveTangent=-3.5186, tangentMode=RCTM_User, time=0.8, value=0}]}}})
      1=&FloatFromCurve(FloatCurve=DataInterface() {propertyValues {Curve {keys [<at 0 1 RCIM_Cubic>, {arriveTangent=-3.5186, interpMode=RCIM_Cubic, leaveTangent=-3.5186, tangentMode=RCTM_User, time=0.5, value=0}]}}})
      2=?
      3=?
      Index=&SpawnGroup()
      Num=::NewEnumerator0
     )
     Translation Contribution:NiagaraFloat=1
     Rotation Contribution:NiagaraFloat=1
    )) {module=BlendSpace002}
   &CurlNoiseForce(Bias Noise=false, Noise Frequency=100, Noise Strength=1000, Pan Noise=false)
   &SolveForcesAndVelocity(Clamp Velocity=false, Limit Acceleration=false)
   &ScaleSpriteSizeBySpeed(Max Scale Factor=<2 30>)
   &ScaleColor(
    Linear Color Curve=DataInterface() {propertyValues {AlphaCurve {keys [<at 0 0>,<at 0.1467 1>,<at 1 0>]}, BlueCurve {keys [<at 0 3>]}, GreenCurve {keys [<at 0 3>]}, RedCurve {keys [<at 0 3>]}}}
    Scale Mode=2
   )
  ]
  rendererClasses [&NiagaraSpriteRendererProperties]
  renderers [{rendererClass=&NiagaraSpriteRendererProperties}]
 }
 Thruster_Light {rendererClasses [], renderers []}
}
summary {
 systemName=NS_Emote_SkiffGrind_Thrusters
 emitters [
  {emitterName=SideThrusters, rendererClasses [&NiagaraMeshRendererProperties, &NiagaraSpriteRendererProperties]}
  {emitterName=BackThrusters, rendererClasses [&NiagaraMeshRendererProperties, &NiagaraSpriteRendererProperties]}
  {emitterName=Sparks, rendererClasses [&NiagaraSpriteRendererProperties]}
  {emitterName=Thruster_Light, rendererClasses []}
  {emitterName=DefenseMatrix, rendererClasses [&NiagaraSpriteRendererProperties, &NiagaraMeshRendererProperties]}
 ]
 userVariables [
  User.bIsFrontEnd {type=NiagaraBool}
  User.bIsFrontEndPreview {type=NiagaraBool}
  User.EngineDelay {type=NiagaraFloat, defaultValue=0.16, description=""}
  User.ExplosionDelay {type=NiagaraFloat, defaultValue=0, description=""}
  User.FeetTrailScale {type=NiagaraFloat, defaultValue=0, description=""}
  User.ForwardDot {type=NiagaraFloat, defaultValue=0, description=""}
  User.HandTrailScale {type=NiagaraFloat, defaultValue=1, description=""}
  User.HitGround {type=NiagaraBool, defaultValue=true}
  User.OwnerLocalVelocity {type=Vector3f, description="", defaultValue {struct=Vector3f, value=<0 0 0>}}
  User.RibbonIDOffset {type=NiagaraInt32}
  User.SkeletalMesh {
   type=NiagaraDataInterfaceSkeletalMesh
   defaultValue {
    struct=NiagaraExt_VariableValue_DataInterface
    value {dataInterface=&self:NiagaraDataInterfaceSkeletalMesh_0}
   }
  }
 ]
}
